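; Captain America shield block volume: a building actor attached to the shield mesh that absorbs hits for its owner.
; Passive and active collision boxes, rescaled on the server, take damage while health is repeatedly reset.
; Damage accumulates toward stagger and decays after a delay; RepNotifies update collision, effects and blocking tags.
; Server overlaps knock back fast vehicles, with cooldown, and convert speed into shield damage, ignoring the owner's base vehicle.
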(event Custom|CheckMovementBaseForVehicle
  (params (NewBaseComponent :type Primitive%20Component%20Object%20Reference) (BoneName :type Name))
  (bind _returnvalue (Components|GetOwner NewBaseComponent))
  ; Check Movement Base For Vehicle
  (Utilities|IsValid NewBaseComponent
    (:Is%20Valid
      (Utilities|IsA(soft) _returnvalue "/Script/FortniteGame.FortVehicleInterface"
        (:True
          (Variables|VehicleBlocking|SetBaseVehicle _returnvalue))
        (:False
          _)))
    (:Is%20Not%20Valid
      _))
  (Variables|VehicleBlocking|SetBaseVehicle 0))

(event Custom|MulticastCollisionUpdate
  (params (CollisionResponse :type ECollisionResponse%20Enum :default "ECR_Ignore"))
  ; Multicast Collision Update
  (Collision|SetCollisionResponsetoChannel (Variables|Default|GetVehicleBlockCollision) "ECC_Vehicle" CollisionResponse))

(event Custom|PutVehicleImpulseOnCooldown
  (params (LaunchActor :type Actor%20Object%20Reference))
  ; Cooldown to prevent Impulse Spam
  (Utilities|Set|Add (Variables|VehicleBlocking|GetLaunchedActor) LaunchActor)
  (Utilities|FlowControl|Delay (ScalableFloat|GetValueatLevel (Variables|ScalableFloats|GetHF_BlockVehicles_Impulse_Cooldown)))
  (Utilities|Set|Remove (Variables|VehicleBlocking|GetLaunchedActor) LaunchActor))

(event EventBeginPlay
  (bind _owningweapon (Variables|Default|GetOwningWeapon))
  (bind _owningweapon_2 (Variables|Default|GetOwningWeapon))
  ;[0.03429 0.3069 0.7899] BeginPlay
  (Utilities|FlowControl|Sequence
    (:then_0
      (bind _asfort_player_pawn (Utilities|Casting|CastToFortPlayerPawn (Variables|Actor|GetInstigator))
        (:then
          (bind _output_get (Variables|Default|SetOwningPlayer _asfort_player_pawn))
          (bind _asb_ashton_shield_weap (Utilities|Casting|CastToB_Ashton_Shield_Weap (Variables|/Script/FortniteGame.FortPawn|GetCurrentWeapon _output_get))
            (:then
              (Variables|Default|SetOwningWeapon _asb_ashton_shield_weap)
              (Transformation|AttachActorToComponent (Variables|/AshtonGameplay/Gameplay/CaptainAmerica/Shield/Actors/B_Ashton_Shield_Weap.B_Ashton_Shield_Weap_C|GetShieldMesh (Variables|Default|GetOwningWeapon)) "None" "SnapToTarget" "KeepRelative" "KeepWorld")
              (Variables|/AshtonGameplay/Gameplay/CaptainAmerica/Shield/Actors/B_Ashton_Shield_Weap.B_Ashton_Shield_Weap_C|SetBlockingBGA self _owningweapon_2)
              (Default|CallBlockingBGAIsSet _owningweapon_2))
            (:CastFailed
              _)))
        (:CastFailed
          _))
      ;;[1 0 0.05961] Doomsday button
      (Actor|DestroyActor))
    (:then_1
      (VehicleBlocking|CheckVehicleBlockEnabled)
      (Networking|SwitchHasAuthority
        (:Authority
          (Variables|/Script/Engine.PrimitiveComponent|SetbGenerateOverlapEvents true (Variables|Default|GetVehicleBlockCollision)))))
    (:then_2
      (CallFunction|SetTeam (Team|GetActorTeam (Game|GetInstigator)))
      (Gameplay|ForceBuildingHealth 5e+04)
      (Transformation|SetRelativeScale3D (Variables|Default|GetActiveBlockCollision) (select (Networking|IsServer) (* (Variables|/Script/Engine.SceneComponent|GetRelativeScale3D (Variables|Default|GetActiveBlockCollision)) (Variables|Default|GetServerScaleMultiplier)) (Variables|/Script/Engine.SceneComponent|GetRelativeScale3D (Variables|Default|GetActiveBlockCollision))))
      (Transformation|SetRelativeScale3D (Variables|Default|GetPassiveBlockCollision) (select (Networking|IsServer) (* (Variables|/Script/Engine.SceneComponent|GetRelativeScale3D (Variables|Default|GetPassiveBlockCollision)) (Variables|Default|GetServerScaleMultiplier)) (Variables|/Script/Engine.SceneComponent|GetRelativeScale3D (Variables|Default|GetPassiveBlockCollision))))
      (CallFunction|ToggleBlockProjectiles true)
      (Networking|SwitchHasAuthority
        (:Authority
          (Utilities|IsValid _owningweapon
            (:Is%20Valid
              (Variables|Default|SetDamageCount (Variables|/AshtonGameplay/Gameplay/CaptainAmerica/Shield/Actors/B_Ashton_Shield_Weap.B_Ashton_Shield_Weap_C|GetInstancedStagger _owningweapon))
              (CallFunction|ResetDecay))))))
    (:then_3
      (Ability|GameplayEffect|AddLooseGameplayTags self (Variables|Default|GetLooseTagsToAddOnSpawn)))
    (:then_4
      (bind _returnvalue (Mutator|GetGameplayMutatorbyClass "/Script/FortniteGame.FortAthenaMutator_WeaponTraceIgnoreFriendlyActors"))
      (Utilities|IsValid _returnvalue
        (:Is%20Valid
          (Class|FortAthenaMutatorWeaponTraceIgnoreFriendlyActors|AddActortoIgnoreifFriendly _returnvalue self))))
    (:then_5
      ;; Listen for movement base updates - determines whether the owner is standing on a vehicle.
      (bind _owningplayer (Variables|Default|GetOwningPlayer)
        (:then
          (EventDispatchers|BindEventtoOnMovementBaseChanged _owningplayer (EventDispatchers|CreateEvent :!event Custom|CheckMovementBaseForVehicle))
          (bind _charactermovement (Variables|/Script/Engine.Character|GetCharacterMovement _owningplayer)
            (:then
              (CallFunction|CheckMovementBaseforVehicle (Utilities|Casting|CastToPrimitiveComponent (Pawn|Components|CharacterMovement|GetMovementBaseObject _charactermovement))))
            (:else)))
        (:else)))))

(event Collision|EventActorBeginOverlap
  (params (OtherActor :type Actor%20Object%20Reference))
  ; Event Actor Begin Overlap (Server only)
  (if (ScalableFloat|GetValueasBool (Variables|ScalableFloats|GetHF_BlockVehicles_Enabled))
    (bind _asfort_vehicle_interface (Utilities|Casting|CastToFortVehicleInterface OtherActor))
    (if (> (Vehicle|GetForwardSpeedKmh _asfort_vehicle_interface) (ScalableFloat|GetValueatLevel (Variables|ScalableFloats|GetHF_BlockVehicles_MinVehicleSpeed)))
      (if (Utilities|Array|ContainsItem (Variables|VehicleBlocking|GetOverlapping%20Vehicles) OtherActor)
        _
        (elif (VehicleBlocking|IsImpulseTargetAppropriate OtherActor)
          (bind (_launchactor _launchvelocity) (Utilities|CalculateLaunchVelocityAwayfromSource (Actor|GetOwner) OtherActor 0 (Variables|VehicleBlocking|GetKnockbackMinYawVehicle) 0 (Variables|VehicleBlocking|GetKnockbackVerticalAngleVehicle) 0 (VehicleBlocking|CalculateVehicleLaunch OtherActor _asfort_vehicle_interface)))
          (if (not (Utilities|Set|ContainsItem (Variables|VehicleBlocking|GetLaunchedActor) _launchactor))
            (CallFunction|PutVehicleImpulseonCooldown _launchactor)
            (CallFunction|MulticastImpulseVehicle _asfort_vehicle_interface _launchvelocity)
            (Utilities|Array|Add (Variables|VehicleBlocking|GetOverlapping%20Vehicles) OtherActor)
            (if (not (GameplayTags|HasAllMatchingGameplayTags (Variables|Default|GetOwningPlayer) (Variables|VehicleBlocking|GetHasVehiclesToBlockTag)))
              (Ability|GameplayEffect|AddLooseGameplayTags (Variables|Default|GetOwningPlayer) (Variables|VehicleBlocking|GetHasVehiclesToBlockTag) true))
            (Game|Damage|ApplyDamage self (VehicleBlocking|CalculateDamagefromVehicle _asfort_vehicle_interface) (Pawn|GetController (Vehicle|GetDriver _asfort_vehicle_interface)) OtherActor)))))))

(event Custom|TestBallisticShieldStagger
  ;[1 0 0.05961] DEBUG
  (if (not (Variables|Default|GetInStaggerState?))
    (Variables|Default|SetInStaggerState? true)))

(event Collision|EventActorEndOverlap
  (params (OtherActor :type Actor%20Object%20Reference))
  (bind _overlapping_vehicles (Variables|VehicleBlocking|GetOverlapping%20Vehicles))
  ; Event Actor End Overlap
  (if (ScalableFloat|GetValueasBool (Variables|ScalableFloats|GetHF_BlockVehicles_Enabled))
    ; Event Actor Begin Overlap (Server only)
    (Utilities|IsA(soft) OtherActor "/Script/FortniteGame.FortVehicleInterface"
      (:True
        (Utilities|Array|RemoveItem _overlapping_vehicles OtherActor)
        (if (Utilities|Array|IsEmpty _overlapping_vehicles)
          (Ability|GameplayEffect|RemoveLooseGameplayTags (Variables|Default|GetOwningPlayer) (Variables|VehicleBlocking|GetHasVehiclesToBlockTag) true))))))

(event Custom|DebugToggleBallisticShieldServerCollision
  ;[1 0 0.05961] DEBUG
  (bind _output_get (Variables|Debug|SetbDebugShowServerCollision (not (Variables|Debug|GetbDebugShowServerCollision))))
  (CallFunction|SetTemporarilyTicking _output_get))

(event EventTick
  (params (DeltaSeconds :type Float%20%28single-precision%29))
  (bind _activeblockcollision (Variables|Default|GetActiveBlockCollision))
  ;[1 0 0.05961] DEBUG
  (if (Math|Boolean|ANDBoolean (Editor|InEditororPIE) (Collision|IsCollisionEnabled (Variables|Default|GetActiveBlockCollision)) (Variables|Default|GetDrawBoxCollision?))
    (bind _returnvalue (FortPhysicsSimulationLibrary|GetBoxExtent _activeblockcollision))
    (Rendering|Debug|DrawDebugBox (Transformation|GetWorldLocation _activeblockcollision) _returnvalue (select (Networking|IsServer) "(R=0.000000,G=0.010310,B=1.000000,A=1.000000)" "(R=1.000000,G=0.011060,B=0.000000,A=1.000000)") (Transformation|GetWorldRotation _activeblockcollision) 0.0 2.0)))

(event Custom|PostDamageTaken
  (params (DamageTaken :type Float%20%28double-precision%29))
  ;[0.03429 0.3069 0.7899] Play impacts and reset health after taking any damage
  (Ability|SendGameplayEventtoActor (Variables|Default|GetOwningPlayer) "(TagName=\"Event.Ashton.Shield.Damage.Blocked\")" (Utilities|Struct|MakeGameplayEventData "(TagName=\"\")" 0 0 0 0 "()" "(GameplayTags=)" "(GameplayTags=)" DamageTaken))
  (GameplayCue|ExecuteGameplayCueOnActor(Burst) self (Variables|FX|GetGC_ShieldImpacted) (Variables|Default|GetDamageParams))
  (Gameplay|ForceBuildingHealth 5e+04))

(event Custom|Multicast_ImpulseVehicle
  (params (Target :type Fort%20Vehicle%20Interface%20Interface) (LinearVelocity :type Vector%20%28by%20ref%29))
  ; Multicast Impulse Vehicle
  (Vehicle|ImpulseVehicle Target LinearVelocity (Math|Vector|VectorZero) false false false "None")
  (GameplayCue|ExecuteGameplayCueOnActor(Burst) self (Variables|FX|GetGC_ShieldImpacted) (Variables|Default|GetDamageParams)))

(event Custom|ToggleBlockProjectiles
  (params (Enabled? :type Boolean))
  ;[0.03429 0.3069 0.7899] Toggle bullet blocking effects
  (Collision|SetCollisionEnabled (Variables|Default|GetPassiveBlockCollision) (select Enabled? "QueryOnly" "NoCollision"))
  ;;[1 0 0.05961] Debug
  (if (Math|Boolean|ANDBoolean (Editor|InEditororPIE) (Variables|Default|GetDrawBoxCollision?) (Variables|Default|GetbActiveBlocking))
    (Development|SetHiddeninGame (Variables|Default|GetPassiveBlockCollision) (not Enabled?) false)))

(event EventOnDamageServer
  (params (Damage :type Float%20%28single-precision%29) (DamageTags :type Gameplay%20Tag%20Container%20Structure%20%28by%20ref%29) (Momentum :type Vector :default "0, 0, 0") (HitInfo :type Hit%20Result%20Structure%20%28by%20ref%29) (InstigatedBy :type Controller%20Object%20Reference) (DamageCauser :type Actor%20Object%20Reference) (EffectContext :type Gameplay%20Effect%20Context%20Handle%20Structure))
  (bind _self self)
  (bind _owningplayer (Variables|Default|GetOwningPlayer))
  (bind _damageparams (Variables|Default|GetDamageParams))
  ;[0.03429 0.3069 0.7899] Damage tracking + visuals
  (Variables|Default|SetDamageParams (CallFunction|GetDamageAbsorbParams HitInfo DamageCauser DamageTags))
  (if (Default|CheckCanBuildStagger?)
    (bind _output_get (Variables|Default|SetDamageCount (+ (CallFunction|CheckDamageMultipliers Damage DamageTags) (Variables|Default|GetDamageCount))))
    (if (>= _output_get (ScalableFloat|GetValueatLevel (Variables|ScalableFloats|GetHF_BlockHealth)))
      (Variables|Default|SetInStaggerState? true)
      (GameplayCue|ExecuteGameplayCueOnActor(Burst) _self (Variables|FX|GetGC_BallisticShieldBreak) _damageparams)
      (else
        (GameplayCue|ExecuteGameplayCueOnActor(Burst) _self (Variables|FX|GetGC_DamageAbsorbed) _damageparams)
        (Ability|SendGameplayEventtoActor _owningplayer (Variables|Default|GetEvent_StaggerAccrued))
        (GameplayCue|ExecuteGameplayCueOnActor(Burst) _owningplayer (Variables|FX|GetGC_StaggerBuilt) (Ability|GameplayCue|MakeGameplayCueParameters (CallFunction|GetDamageCountasPercentage)))))
    (CallFunction|ResetDecay)
    (CallFunction|PostDamageTaken Damage)
    (else
      (CallFunction|PostDamageTaken Damage))))

(event Custom|ResetDecay
  ;[0.2248 0.7344 0.0859] Handle damage buildup decay when user has not taken damage for a duration
  (CallFunction|CheckandClearTimers)
  (bind _returnvalue (Utilities|Time|SetTimerbyEvent (delegate Custom|BeginDamageDecay) (ScalableFloat|GetValueatLevel (Variables|Decay|GetHF_DecayStartTime))))
  (Variables|Decay|SetDecayStartTimer _returnvalue))

(event Custom|BeginDamageDecay
  ;[0.2248 0.7344 0.0859] Handle damage buildup decay when user has not taken damage for a duration
  (bind _returnvalue (Utilities|Time|SetTimerbyEvent (delegate Custom|DecayDamage) (ScalableFloat|GetValueatLevel (Variables|Decay|GetHF_DecayRate)) true))
  (Variables|Decay|SetDecayTimer _returnvalue))

(event Custom|DecayDamage
  (bind _hf_damagedecaypertick (Variables|Decay|GetHF_DamageDecayPerTick))
  (bind _returnvalue (ScalableFloat|GetValueatLevel _hf_damagedecaypertick))
  (bind _damagecount (Variables|Default|GetDamageCount))
  (bind _returnvalue_2 (- _damagecount _returnvalue))
  ;[0.2248 0.7344 0.0859] Handle damage buildup decay when user has not taken damage for a duration
  (if (<= _returnvalue_2 0)
    (Variables|Default|SetDamageCount 0.0)
    (CallFunction|CheckandClearTimers)
    (else
      (Variables|Default|SetDamageCount _returnvalue_2))))

(event Custom|UpdateActiveBlockCollision
  (bind _activeblockcollision (Variables|Default|GetActiveBlockCollision))
  ;[0 0.3311 1] Toggle full frontal blocking based on activation and stagger state.
  (Collision|SetCollisionEnabled _activeblockcollision (select (and (Variables|Default|GetbActiveBlocking) (not (Variables|Default|GetInStaggerState?))) "QueryOnly" "NoCollision"))
  ;;[1 0 0.07858] DEBUG
  (Development|SetHiddeninGame _activeblockcollision (not (Math|Boolean|ANDBoolean (Variables|Default|GetbActiveBlocking) (not (Variables|Default|GetInStaggerState?)) (Variables|Default|GetDrawBoxCollision?) (Editor|InEditororPIE))) false)
  (if (Networking|IsServer)
    (Networking|ForceNetUpdate)))

(event EventEndPlay
  (params (EndPlayReason :type EEndPlayReason%20Enum :default "Destroyed"))
  (Utilities|FlowControl|Sequence
    (:then_0
      (bind _returnvalue (Mutator|GetGameplayMutatorbyClass "/Script/FortniteGame.FortAthenaMutator_WeaponTraceIgnoreFriendlyActors"))
      (Utilities|IsValid _returnvalue
        (:Is%20Valid
          (Class|FortAthenaMutatorWeaponTraceIgnoreFriendlyActors|RemoveActortoIgnoreifFriendly _returnvalue self))))
    (:then_1
      (bind _owningplayer (Variables|Default|GetOwningPlayer)
        (:then
          (EventDispatchers|UnbindallEventsfromOnMovementBaseChanged _owningplayer))
        (:else)))))

(fn CalculateDamageFromVehicle
  (params (Vehicle :type Fort%20Vehicle%20Interface%20Interface))
  (bind _minspeedforvehicledamage (Variables|VehicleBlocking|GetMinSpeedForVehicleDamage))
  (bind _returnvalue (ScalableFloat|GetValueatLevel _minspeedforvehicledamage))
  (if (< (Vehicle|GetPreviousForwardSpeedKmh Vehicle) _returnvalue)
    (return 0.0)
    (else
      (return (Math|Float|MapRangeClamped 0.0 _returnvalue (ScalableFloat|GetValueatLevel (Variables|VehicleBlocking|GetMaxSpeedForVehicleDamage)) (ScalableFloat|GetValueatLevel (Variables|VehicleBlocking|GetMinDamageFromVehicle)) (ScalableFloat|GetValueatLevel (Variables|VehicleBlocking|GetMaxDamageFromVehicle)))))))

(fn CalculateVehicleLaunch
  (params (Actor :type Actor%20Object%20Reference) (FortVehicleInterace :type Fort%20Vehicle%20Interface%20Interface))
  (return (Math|Float|SelectFloat (Math|Vector|VectorLength (+ (* (Vehicle|GetVehicleVelocity FortVehicleInterace) -1.4) (* (Transformation|GetActorUpVector Actor) 3.5e+02))) (Math|Vector|VectorLength (* (Transformation|GetActorForwardVector Actor) -7.5e+02)) (>= (Math|Vector|VectorLength (Vehicle|GetVehicleVelocity FortVehicleInterace)) 2.5e+02))))

(fn CheckAndClearTimers
  (bind _decaystarttimer (Variables|Decay|GetDecayStartTimer))
  (bind _decaytimer (Variables|Decay|GetDecayTimer))
  ;[0.778 0.04391 1] Wipe timers when damage is taken
  (Utilities|FlowControl|Sequence
    (:then_0
      (if (Utilities|Time|IsTimerActivebyHandle _decaystarttimer)
        (Utilities|Time|ClearandInvalidateTimerbyHandle _decaystarttimer)))
    (:then_1
      (if (Utilities|Time|IsTimerActivebyHandle _decaytimer)
        (Utilities|Time|ClearandInvalidateTimerbyHandle _decaytimer)))
    (:then_2
      (return))))

(fn CheckCanBuildStagger?
  ;[0.778 0.04391 1] Check player state and stagger state
  (return (and (not (Variables|Default|GetInStaggerState?)) (Utilities|Array|ContainsItem (Variables|Default|GetDamagableActionStates) (Variables|/AshtonGameplay/Gameplay/CaptainAmerica/Shield/Actors/B_Ashton_Shield_Weap.B_Ashton_Shield_Weap_C|GetShieldActionState (Variables|Default|GetOwningWeapon))))))

(fn CheckDamageMultipliers
  (params (Damage :type Float%20%28double-precision%29) (Tags :type Gameplay%20Tag%20Container%20Structure))
  ;[0.03429 0.3069 0.7899] Iterate our damage multiplier data and see if this instance of damage should be multiplied
  (for _array_element (Variables|Default|GetDamageMultiplers)
    (bind (_array_element_damagetags_3_788373e142b2fe9b40b10f82cada4b16 _array_element_multiplier_7_44facfda418066749597d298806026bd) (Utilities|Struct|BreakFortBShieldMultiplierData _array_element))
    (if (GameplayTags|HasAnyTags (Variables|Actor|GetTags) _array_element_damagetags_3_788373e142b2fe9b40b10f82cada4b16 true)
      (return (* (Variables|GetDamage) (ScalableFloat|GetValueatLevel _array_element_multiplier_7_44facfda418066749597d298806026bd)))))
  (return Damage))

(fn CheckVehicleBlockEnabled
  (if (not (ScalableFloat|GetValueasBool (Variables|ScalableFloats|GetHF_BlockVehicles_Enabled)))
    (Components|DestroyComponent (Variables|Default|GetVehicleBlockCollision))))

(fn EndStagger
  (Variables|Default|SetInStaggerState? false)
  (Variables|Default|SetDamageCount 0.0))

(fn GetDamageAbsorbParams
  (params (Hit_bBlockingHit :type Boolean) (Hit_bInitialOverlap :type Boolean) (Hit_Time :type Float%20%28single-precision%29) (Hit_Distance :type Float%20%28single-precision%29) (Hit_Location :type Vector :default "0, 0, 0") (Hit_ImpactPoint :type Vector :default "0, 0, 0") (Hit_Normal :type Vector :default "0, 0, 0") (Hit_ImpactNormal :type Vector :default "0, 0, 0") (Hit_PhysMat :type Physical%20Material%20Object%20Reference) (Hit_HitActor :type Actor%20Object%20Reference) (Hit_HitComponent :type Primitive%20Component%20Object%20Reference) (Hit_HitBoneName :type Name :default "None") (Hit_BoneName :type Name :default "None") (Hit_HitItem :type Integer) (Hit_ElementIndex :type Integer) (Hit_FaceIndex :type Integer) (Hit_TraceStart :type Vector :default "0, 0, 0") (Hit_TraceEnd :type Vector :default "0, 0, 0") (DamageCauser :type Actor%20Object%20Reference) (DamageTags :type Gameplay%20Tag%20Container%20Structure))
  (return (Ability|GameplayCue|MakeGameplayCueParameters :NormalizedMagnitude (CallFunction|GetDamageCountasPercentage) :RawMagnitude 0.0 :AggregatedSourceTags DamageTags :Location Hit_Location :Normal Hit_ImpactNormal :EffectCauser DamageCauser :SourceObject (Variables|Default|GetOwningPlayer) :GameplayEffectLevel 1 :AbilityLevel 1 :bReplicateLocationWhenUsingMinimalRepProxy false)))

(fn GetDamageCountAsPercentage
  (return (Math|Float|Clamp(Float) (Math|Float|SafeDivide (Variables|Default|GetDamageCount) (ScalableFloat|GetValueatLevel (Variables|ScalableFloats|GetHF_BlockHealth))))))

(fn IsImpulseTargetAppropriate
  (params (Actor :type Object%20Reference))
  (return (and (not (and (== Actor (Variables|VehicleBlocking|GetBaseVehicle)) (Utilities|IsValid (Variables|VehicleBlocking|GetBaseVehicle) :!class /Script/Engine.KismetSystemLibrary))) (not (GameplayTags|HasMatchingGameplayTag (Variables|Default|GetOwningPlayer) "(TagName=\"Gameplay.Action.Player.AthenaInVehicle\")")))))

(fn IsImpulseTargetPawnBaseVehicle
  (params (Actor :type Object%20Reference))
  (return (and (== Actor (Variables|VehicleBlocking|GetBaseVehicle)) (Utilities|IsValid (Variables|VehicleBlocking|GetBaseVehicle) :!class /Script/Engine.KismetSystemLibrary))))

(fn OnRep_bActiveBlocking
  (bind _owningplayer (Variables|Default|GetOwningPlayer))
  (Utilities|FlowControl|Sequence
    (:then_0
      (CallFunction|UpdateActiveBlockCollision))
    (:then_1
      (Utilities|IsValid _owningplayer
        (:Is%20Valid
          (Utilities|IsValid (Ability|GetAbilitySystemComponent _owningplayer)
            (:Is%20Valid
              (Fortnite|ApplyGameplayEffectsToActor (Variables|Default|GetOwningPlayer) (Utilities|Array|MakeArray (select (Variables|Default|GetbActiveBlocking) (Variables|Default|GetGE_BlockingTag_Add) (Variables|Default|GetGE_BlockingTag_Remove))) (Transformation|GetActorLocation (Variables|Default|GetOwningPlayer)) false 0 (Ability|GetAbilitySystemComponent (Variables|Default|GetOwningPlayer))))))))))

(fn OnRep_DamageCount
  (bind _owningweapon (Variables|Default|GetOwningWeapon))
  (bind _ns_ashton_shield_damagestate (Variables|/AshtonGameplay/Gameplay/CaptainAmerica/Shield/Actors/B_Ashton_Shield_Weap.B_Ashton_Shield_Weap_C|GetNS_Ashton_Shield_DamageState _owningweapon))
  (Utilities|FlowControl|Sequence
    (:then_0
      (if (and (> (Variables|Default|GetDamageCount) 0) (> (Variables|Default|GetDamageCount) (Variables|Default|GetDamageCount_Previous)))
        (Utilities|IsValid _owningweapon
          (:Is%20Valid
            (Components|Activation|Activate _ns_ashton_shield_damagestate false)
            (Niagara|SetNiagaraVariable(Float) _ns_ashton_shield_damagestate "DamageState" (CallFunction|GetDamageCountasPercentage))))))
    (:then_1
      (Variables|Default|SetDamageCount_Previous (Variables|Default|GetDamageCount)))
    (:then_2
      (Default|CallOnStaggerUpdated))))

(fn OnRep_InStaggerState?
  (CallFunction|UpdateActiveBlockCollision)
  (if (Variables|Default|GetInStaggerState?)
    (Ability|SendGameplayEventtoActor (Variables|Default|GetOwningPlayer) "(TagName=\"Event.Paprika.BallisticShield.Player.Stagger\")")
    (CallFunction|ToggleBlockProjectiles false)
    (else
      (Gameplay|ForceBuildingHealth 1e+04)
      (CallFunction|ToggleBlockProjectiles true)
      (Development|PrintString "block reenabled"))))

(fn OnRep_ResetDecay?
  (if (Variables|Decay|GetResetDecay?)
    (CallFunction|ResetDecay)
    (else
      (CallFunction|CheckandClearTimers)))
  (return))

(fn ConstructionScript)
