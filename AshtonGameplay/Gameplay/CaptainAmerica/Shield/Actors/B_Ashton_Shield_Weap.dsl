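; Captain America's throwable, blocking shield weapon (Ashton), which manages equip, holstering while the projectile is out, and Idling, Blocking or Charging states.
; State changes drive gameplay cues, charge audio and server-side mesh poses that toggle the BlockVolume BGA's active blocking.
; On unequip, it saves stagger buildup from that BGA and can decay it by timer.
; The autonomous client polls for line-of-sight homing targets to feed the target-state component.
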(event EventBeginPlay
  ; MetaSound Pages test
  (|CallParentFunction)
  (CallFunction|SetupMetaSoundPages))

(event FortWeapon|EventOnEquippedAndAttached
  (bind _weaponowner (Variables|Default|GetWeaponOwner))
  (|CallParentFunction)
  (Utilities|FlowControl|Sequence
    (:then_0
      (bind _asfort_pawn (Utilities|Casting|CastToFortPawn (Game|GetInstigator)))
      (Variables|Default|SetWeaponOwner _asfort_pawn)
      (Homing|StartHomingTargetCheck)
      (CallFunction|ClearBlockingShieldUnequippedTimers)
      (bind _projectile_replicated (Variables|Default|GetProjectile%20Replicated)
        (:then
          (Weapon|SetHolsterWeapon (Variables|Default|GetWeaponOwner) _projectile_replicated true false "ashton_shield") ; If the projectile is out,  holster the weapon
        )
        (:else)))
    (:then_1
      ;[0.03429 0.3069 0.7899] Attach shield to right hand (or root in case of server)
      (Transformation|AttachComponentToComponent (Variables|Default|GetShieldMesh) (Variables|/Script/Engine.Character|GetMesh (Variables|Default|GetWeaponOwner)) (select (Networking|IsServer) "root" "hand_attach_r") "KeepRelative" "KeepRelative" "KeepRelative" true)
      ;[0.03429 0.3069 0.7899] Adjust server shield collision
      (if (Networking|IsServer)
        (CallFunction|UpdateServerShieldLocationandRotation false))
      (CallFunction|SetShieldActionState "Idling")
      ;[0.03429 0.3069 0.7899] Listen to charge event and set player state
      (Ability|Async|EndAction (Variables|Default|GetAsyncAction_EventChargeStarted))
      (bind _asynctaskproxy (Ability|Async|WaitGameplayEventToActor _weaponowner (Variables|Default|GetEvent_ChargeStarted))
        (:then
          (Variables|Default|SetAsyncAction_EventChargeStarted _asynctaskproxy)
          (Ability|Async|EndAction (Variables|Default|GetAsyncAction_EventChargeEnded))
          (bind _asynctaskproxy_2 (Ability|Async|WaitGameplayEventToActor _weaponowner (Variables|Default|GetEvent_ChargeEnded))
            (:then
              (Variables|Default|SetAsyncAction_EventChargeEnded _asynctaskproxy_2))
            (:EventReceived
              (CallFunction|StopChargeAudio)
              (CallFunction|SetShieldActionState "Idling"))))
        (:EventReceived
          (CallFunction|SetShieldActionState "Charging"))))
    (:then_2
      (Rendering|SetVisibility (Variables|Default|GetShieldComponent) true false)
      (CallFunction|SetShieldActionState "Idling"))))

(event FortWeapon|EventOnUnEquippedAndDetached
  (|CallParentFunction)
  (Utilities|FlowControl|Sequence
    (:then_0
      (Rendering|SetVisibility (Variables|Default|GetShieldComponent) false false)
      (Networking|SwitchHasAuthority
        (:Authority
          (CallFunction|SaveInstancedStagger))
        (:Remote
          _))
      (CallFunction|SetShieldActionState "Idling"))
    (:then_1
      (Ability|Async|EndAction (Variables|Default|GetAsyncAction_EventChargeStarted))
      (Ability|Async|EndAction (Variables|Default|GetAsyncAction_EventChargeEnded)))
    (:then_2
      (bind _projectile_replicated (Variables|Default|GetProjectile%20Replicated)
        (:then
          (bind _weaponowner (Variables|Default|GetWeaponOwner)
            (:then
              (Weapon|SetHolsterWeapon _weaponowner _projectile_replicated false false "ashton_shield") ; Projectile is the holster source (if the projectile doesn't exist, the weapon won't be holstered)
            )
            (:else)))
        (:else)))
    (:then_3
      (Utilities|Time|ClearandInvalidateTimerbyHandle (Variables|Homing|GetTimer_UpdateHomingCurrentTarget)))))

(event Custom|BlockingShield_RecoveryWhileUnequipped
  (params (BeginDelay :type Float%20%28double-precision%29))
  ;[0.2248 0.7344 0.0859] Blocking shield: Handle damage buildup decay when the shield is unequipped, when it's equipped, it's handled in BGA_Ashton_Shield_BlockVolume
  (CallFunction|ClearBlockingShieldUnequippedTimers)
  (if (> BeginDelay 0)
    (bind _returnvalue (Utilities|Time|SetTimerbyEvent (delegate Custom|BlockingShield_BeginDamageDecay) BeginDelay))
    (Variables|Default|SetBlockingShield_DecayStartTimer _returnvalue)
    (else
      (bind _returnvalue_2 (Utilities|Time|SetTimerbyEvent (delegate Custom|BlockingShield_DecayDamage) (ScalableFloat|GetValueatLevel (Variables|ScalableFloats|GetHF_DecayRate)) true))
      (Variables|Default|SetBlockingShield_DecayTimer _returnvalue_2))))

(event Custom|BlockingShield_BeginDamageDecay
  ;[0.2248 0.7344 0.0859] Blocking shield: Handle damage buildup decay when the shield is unequipped, when it's equipped, it's handled in BGA_Ashton_Shield_BlockVolume
  (bind _returnvalue (Utilities|Time|SetTimerbyEvent (delegate Custom|BlockingShield_DecayDamage) (ScalableFloat|GetValueatLevel (Variables|ScalableFloats|GetHF_DecayRate)) true))
  (Variables|Default|SetBlockingShield_DecayTimer _returnvalue))

(event Custom|ServerCancelTacSprint
  ;[1 0.6954 0.3789] Cancel Tactical Sprint Ability
  ; For use by GA_Ashton_Shield_CheckIfTryingToBlock and allows cancelling tac sprint ability if trying to block while tac sprinting. (The block GA itself cannot cancel the tac sprint as it does not activate since the weapon is holstered while tac sprinting.)
  (Networking|SwitchHasAuthority
    (:Authority
      (bind _weaponowner (Variables|Default|GetWeaponOwner)
        (:then
          (Utilities|IsValid (Ability|GetFortAbilitySystemComponent _weaponowner)
            (:Is%20Valid
              (Tags|CancelAbilitiesWithTags (Ability|GetFortAbilitySystemComponent _weaponowner) (GameplayTags|MakeLiteralGameplayTagContainer "(GameplayTags=((TagName=\"Gameplay.Action.Player.TacticalSprint\")))")))))
        (:else)))))

(event Custom|BlockingShield_DecayDamage
  (bind _hf_damagedecaypertick (Variables|ScalableFloats|GetHF_DamageDecayPerTick))
  (bind _returnvalue (ScalableFloat|GetValueatLevel _hf_damagedecaypertick))
  (bind _instancedstagger (Variables|Default|GetInstancedStagger))
  (bind _returnvalue_2 (- _instancedstagger _returnvalue))
  ;[0.2248 0.7344 0.0859] Blocking shield: Handle damage buildup decay when the shield is unequipped, when it's equipped, it's handled in BGA_Ashton_Shield_BlockVolume
  (if (<= _returnvalue_2 0)
    (Variables|Default|SetInstancedStagger 0.0)
    (CallFunction|ClearBlockingShieldUnequippedTimers)
    (else
      (Variables|Default|SetInstancedStagger _returnvalue_2))))

(event EventOnSetTargeting
  (params (bNewIsTargeting :type Boolean))
  ;[0.03429 0.3069 0.7899] Set player state on aim down sights
  (|CallParentFunction bNewIsTargeting)
  (CallFunction|SetShieldActionState (select bNewIsTargeting "Blocking" "Idling")))

(event Custom|NewLockTarget
  (params (LockOnTarget :type Actor%20Object%20Reference))
  (if (Pawn|IsLocallyControlled (Variables|Actor|GetInstigator))
    (Projectile|AddGameplayCue%28Looping%29Local LockOnTarget (Variables|Default|GetTagLockOnFX) (Ability|GameplayCue|MakeGameplayCueParameters :NormalizedMagnitude 0.0 :RawMagnitude 0.0 :Location "0,0,0" :Normal "0,0,0" :Instigator (Variables|Actor|GetInstigator) :GameplayEffectLevel 1 :AbilityLevel 1 :bReplicateLocationWhenUsingMinimalRepProxy false))
    (Utilities|Array|AddUnique (Variables|Default|GetOldLockOnTargets) LockOnTarget)))

(event Custom|UpdateServerShieldLocationAndRotation
  (params (bIsInBlockingState :type Boolean))
  (bind _blockingbga (Variables|Default|GetBlockingBGA))
  (if bIsInBlockingState
    (bind (_direction _movetoblock) (Timeline|Timeline_MoveServerShieldToBlock PlayFromStart)
      (:Update
        (Transformation|SetRelativeLocationAndRotation (Variables|Default|GetShieldMesh) (Math|Vector|Lerp(Vector) (Variables|Default|GetServerShieldLocation_Idle) (Variables|Default|GetServerShieldLocation_Blocking) _movetoblock) (Math|Rotator|Lerp(Rotator) (Variables|Default|GetServerShieldRotation_Idle) (Variables|Default|GetServerShieldRotation_Blocking) _movetoblock) false false))
      (:Finished
        _))
    (else
      (bind (_direction_2 _movetoidle) (Timeline|Timeline_MoveServerShieldToIdle PlayFromStart)
        (:Update
          (Transformation|SetRelativeLocationAndRotation (Variables|Default|GetShieldMesh) (Math|Vector|Lerp(Vector) (Variables|Default|GetServerShieldLocation_Idle) (Variables|Default|GetServerShieldLocation_Blocking) (- 1.0 _movetoidle)) (Math|Rotator|Lerp(Rotator) (Variables|Default|GetServerShieldRotation_Idle) (Variables|Default|GetServerShieldRotation_Blocking) (- 1.0 _movetoidle)) false false))
        (:Finished
          _))))
  (Transformation|SetRelativeLocationAndRotation (Variables|Default|GetShieldMesh) (select bIsInBlockingState (Variables|Default|GetServerShieldLocation_Blocking) (Variables|Default|GetServerShieldLocation_Idle)) (select bIsInBlockingState (Variables|Default|GetServerShieldRotation_Blocking) (Variables|Default|GetServerShieldRotation_Idle)) false false)
  (Variables|Default|SetbBlockingConfirmed bIsInBlockingState)
  (Utilities|IsValid _blockingbga
    (:Is%20Valid
      (Variables|/AshtonGameplay/Gameplay/CaptainAmerica/Shield/Actors/BGA_Ashton_Shield_BlockVolume.BGA_Ashton_Shield_BlockVolume_C|SetbActiveBlocking bIsInBlockingState _blockingbga))))

(event Custom|ClearLockTargets
  (for _array_element (Variables|Default|GetOldLockOnTargets)
    (Utilities|IsValid _array_element
      (:Is%20Valid
        (Projectile|RemoveGameplayCue%28Looping%29Local _array_element (Variables|Default|GetTagLockOnFX) (Ability|GameplayCue|MakeGameplayCueParameters 0.0)))))
  (Utilities|Array|Clear (Variables|Default|GetOldLockOnTargets)))

(event Custom|PlayStateSoundLibraryEvent_Client
  (params (GameplayTag :type Gameplay%20Tag%20Structure))
  (Networking|CanExecuteCosmeticEvents
    (:True
      (if (GameplayTags|IsGameplayTagValid GameplayTag)
        (SoundLibrary|PlaySoundSimple (WorldSubsystems|GetSoundLibrarySubsystem) self (Utilities|Struct|MakeSimpleSoundSettings GameplayTag true) :!class /Script/SoundLibrary.SoundLibrarySubsystem)))))

(event EventOnWeaponVisibilityChanged
  (params (bVisible :type Boolean) (bSetForLocalControllerOnly :type Boolean))
  (if (not bSetForLocalControllerOnly)
    (Class|BGAAshtonShieldBlockVolume|ToggleBlockProjectiles (Variables|Default|GetBlockingBGA) bVisible)))

(event Custom|SetProjectile
  (params (Projectile :type B%20Prj%20Ashton%20Shield%20Object%20Reference))
  ;[0 0.1718 1] SetProjectile
  (Variables|Default|SetProjectile%20Replicated Projectile)
  (Variables|Homing|SetHomingTarget 0) ; Clear homing target after proj is spawned
)

(event Custom|HandleProjectileHit
  ;[0.1669 0 1] HandleProjectileHit(): calls Client_HandleProjectileHit()
  ; Client_HandleProjectileHit(): plays a timeline that shows the ReticleHitColor going from 1 to 0.
  (Networking|SwitchHasAuthority
    (:Authority
      (CallFunction|ClientHandleProjectileHit))))

(event Custom|Client%20HandleProjectileHit
  ;[0.1669 0 1] HandleProjectileHit(): calls Client_HandleProjectileHit()
  ; Client_HandleProjectileHit(): plays a timeline that shows the ReticleHitColor going from 1 to 0.
  (Networking|CanExecuteCosmeticEvents
    (:True
      (bind (_direction _hitpercent) (Timeline|Timeline PlayFromStart)
        (:Update
          (Variables|Default|SetReticleHitPercent _hitpercent))
        (:Finished
          (Variables|Default|SetReticleHitPercent 0.0))))))

(fn ClearBlockingShieldUnequippedTimers
  (bind _blockingshield_decaystarttimer (Variables|Default|GetBlockingShield_DecayStartTimer))
  (bind _blockingshield_decaytimer (Variables|Default|GetBlockingShield_DecayTimer))
  (Utilities|FlowControl|Sequence
    (:then_0
      (if (Utilities|Time|IsTimerActivebyHandle _blockingshield_decaystarttimer)
        (Utilities|Time|ClearandInvalidateTimerbyHandle _blockingshield_decaystarttimer)))
    (:then_1
      (if (Utilities|Time|IsTimerActivebyHandle _blockingshield_decaytimer)
        (Utilities|Time|ClearandInvalidateTimerbyHandle _blockingshield_decaytimer)))))

(fn GetAshtonShieldData
  (bind _blockingbga (Variables|Default|GetBlockingBGA))
  (Utilities|IsValid _blockingbga
    (:Is%20Valid
      (return (Variables|/AshtonGameplay/Gameplay/CaptainAmerica/Shield/Actors/BGA_Ashton_Shield_BlockVolume.BGA_Ashton_Shield_BlockVolume_C|GetHF_BlockHealth _blockingbga) (Variables|/AshtonGameplay/Gameplay/CaptainAmerica/Shield/Actors/BGA_Ashton_Shield_BlockVolume.BGA_Ashton_Shield_BlockVolume_C|GetDamageCount _blockingbga)))
    (:Is%20Not%20Valid
      (return 0 0.0))))

(fn GetBallisticShieldData
  (bind _staggerbuilduppercent (CallFunction|GetStaggerBuildupasPercentage))
  (return (Variables|Default|GetShieldActionState) _staggerbuilduppercent))

(fn GetStaggerBuildupAsPercentage
  (Utilities|IsValid (Variables|Default|GetBlockingBGA)
    (:Is%20Valid
      (return (Class|BGAAshtonShieldBlockVolume|GetDamageCountasPercentage (Variables|Default|GetBlockingBGA))))
    (:Is%20Not%20Valid
      (return 0.0))))

(fn GetWeaponSpecificAttachedActors
  (return (Utilities|Array|MakeArray (Variables|Default|GetBlockingBGA))))

(fn LOSCheck
  (params (OriginActor :type Actor%20Object%20Reference) (TargetActor :type Actor%20Object%20Reference))
  (Utilities|Array|Clear (LocalVariables|GetIgnoreActorsArray))
  (Utilities|Array|AddUnique (LocalVariables|GetIgnoreActorsArray) (Variables|GetOriginActor))
  (bind _outhit (Collision|LineTraceByChannel (Transformation|GetActorLocation (Variables|GetOriginActor)) (Transformation|GetActorLocation (Variables|GetTargetActor)) "TraceTypeQuery1" false (LocalVariables|GetIgnoreActorsArray)))
  (if (Utilities|Equal(Object) (.HitActor (Collision|BreakHitResult _outhit)) (Variables|GetTargetActor))
    (return true)
    (else
      (return false))))

(fn OnRep_bBlockingConfirmed
  (Default|CallOnBlockingStateConfirmed (Variables|Default|GetShieldActionState)))

(fn OnRep_HomingTarget
  (Variables|Homing|GetHomingTarget
    (:then)
    (:else)))

(fn OnRep_ProjectileReplicated
  (Default|CallOnShieldProjectileReplicated))

(fn PlayStateSoundLibraryEvent
  (params (Event%20Name :type Gameplay%20Tag%20Structure))
  (CallFunction|PlayStateSoundLibraryEventClient Event%20Name))

(fn SaveInstancedStagger
  (bind _blockingbga (Variables|Default|GetBlockingBGA))
  ;[0.4546 0.06953 0.6667] Save our stagger value before switching away from the weapon
  (Utilities|IsValid _blockingbga
    (:Is%20Valid
      (bind _output_get (Variables|Default|SetInstancedStagger (Math|Float|Clamp(Float) (Variables|/AshtonGameplay/Gameplay/CaptainAmerica/Shield/Actors/BGA_Ashton_Shield_BlockVolume.BGA_Ashton_Shield_BlockVolume_C|GetDamageCount _blockingbga) 0.0 (ScalableFloat|GetValueatLevel (Variables|/AshtonGameplay/Gameplay/CaptainAmerica/Shield/Actors/BGA_Ashton_Shield_BlockVolume.BGA_Ashton_Shield_BlockVolume_C|GetHF_BlockHealth _blockingbga)))))
      (Development|PrintString (Utilities|String|Append "Stagger saved: " (Utilities|String|ToString(Float) _output_get)))
      (if (<= (Variables|Default|GetInstancedStagger) 0)
        _
        (elif (ScalableFloat|GetValueasBool (Variables|ScalableFloats|GetHF_DecayRecoveryWhenUnequipped))
          (bind _blockingbga_2 (Variables|Default|GetBlockingBGA)
            (:then
              (if (Utilities|Time|IsValidTimerHandle (Variables|/AshtonGameplay/Gameplay/CaptainAmerica/Shield/Actors/BGA_Ashton_Shield_BlockVolume.BGA_Ashton_Shield_BlockVolume_C|GetDecayStartTimer _blockingbga_2))
                (CallFunction|BlockingShieldRecoveryWhileUnequipped (Utilities|Time|GetTimerRemainingTimebyHandle (Variables|/AshtonGameplay/Gameplay/CaptainAmerica/Shield/Actors/BGA_Ashton_Shield_BlockVolume.BGA_Ashton_Shield_BlockVolume_C|GetDecayStartTimer _blockingbga_2)))
                (else
                  (CallFunction|BlockingShieldRecoveryWhileUnequipped 0.0)))
              (return))
            (:else
              _))))
      (return))))

(fn SetShieldActionState
  (params (NewState :type EBallisticShieldPlayerActionState%20Enum :default "Idling"))
  (bind _weaponowner (Variables|Default|GetWeaponOwner))
  (bind _tempparams (LocalVariables|GetTempParams))
  ;[0.4895 0.1317 1] Update shield state, remove and apply looping cues based on new state
  (Default|CallOnActionStateChanged (Variables|Default|GetShieldActionState) (Variables|GetNewState))
  (bind _output_get (Variables|Default|SetShieldActionState (Variables|GetNewState)))
  (GameplayCue|RemoveGameplayCueOnActor(Looping) _weaponowner (Variables|Default|GetGC_Blocking) _tempparams)
  (GameplayCue|RemoveGameplayCueOnActor(Looping) _weaponowner (Variables|Default|GetGC_Charging) _tempparams)
  (switch Utilities|FlowControl|Switch|SwitchonEBallisticShieldPlayerActionState _output_get
    (:Idling
      _)
    (:Blocking
      (GameplayCue|AddGameplayCueOnActor(Looping) (Variables|Default|GetWeaponOwner) (Utilities|Select :Index (Variables|Default|GetShieldActionState) :Blocking (Variables|Default|GetGC_Blocking) :Charging (Variables|Default|GetGC_Charging)) (LocalVariables|GetTempParams)))
    (:Charging
      (CallFunction|PlayStateSoundLibraryEvent "(TagName=\"SoundLibrary.Ashton.Shield.Charge.Intro\")")
      (return)))
  (if (Networking|IsServer)
    (CallFunction|UpdateServerShieldLocationandRotation (Utilities|Select :Index (Variables|Default|GetShieldActionState) :Idling false :Blocking true :Charging true)))
  (return))

(fn SetupMetaSoundPages
  (bind _returnvalue (SoundLibrary|GetSoundLibraryComponent "(TagName=\"SoundLibrary.Weapon\")"))
  (bind _mspagessoundlibrary (Variables|Default|GetMSPagesSoundLibrary))
  (Utilities|IsValid _returnvalue
    (:Is%20Valid
      (if (ScalableFloat|GetValueasBool (Variables|Default|GetHF_EnableMetaSoundPages))
        (SoundLibrary|AddLibrary _returnvalue _mspagessoundlibrary :!class /Script/SoundLibrary.SoundLibraryComponent)
        (else
          (SoundLibrary|RemoveLibrary _returnvalue _mspagessoundlibrary :!class /Script/SoundLibrary.SoundLibraryComponent))))))

(fn StartHomingTargetCheck
  (if (Utilities|Enum|Equal(Enum) (Networking|GetLocalRole (Game|GetInstigator)) "ROLE_AutonomousProxy")
    (bind _returnvalue (Utilities|Time|SetTimerbyFunctionName 0 "UpdateHomingCurrentTarget" (Variables|Homing|GetHomingReticleCheckFrequency) true true))
    (Variables|Homing|SetTimer_UpdateHomingCurrentTarget _returnvalue)))

(fn StopChargeAudio
  ; Moving shield end charge audio here instead so it breaks from the ballistic shield audio implementation to prevent issues
  (if (not (Networking|IsServer))
    (bind _returnvalue (Audio|SpawnSoundAttached (Variables|Default|GetShieldChargeEndSound) (Variables|Visual|GetWeaponMesh)))
    (bind _weaponowner (Variables|Default|GetWeaponOwner)
      (:then
        (Audio|Parameter|SetBooleanParameter _returnvalue "isFriendly" (select (Pawn|IsLocallyViewed _weaponowner) false (select (Team|OnLocalPlayersTeam _weaponowner) true false)) :!class /Script/AudioExtensions.AudioParameterControllerInterface))
      (:else
        _)))
  (return))

(fn UpdateHomingCurrentTarget
  (bind _weaponowner (Variables|Default|GetWeaponOwner))
  (bind _returnvalue (Fortnite|GetFortPlayerControllerfromActor _weaponowner))
  (bind _fortweaponcomponent_targetstate (Variables|Default|GetFortWeaponComponent_TargetState))
  (bind _fortweaponcomponent_targetstate_2 (Variables|Default|GetFortWeaponComponent_TargetState))
  (Variables|Default|GetWeaponOwner
    (:then
      (Utilities|IsValid _returnvalue
        (:Is%20Valid
          (bind _returnvalue_2 (Fortnite|GetAimedAtEnemy _returnvalue (ScalableFloat|GetValueatLevel (Variables|ScalableFloats|GetRow_Homing_RectHalfWidth)) (ScalableFloat|GetValueatLevel (Variables|ScalableFloats|GetRow_Homing_RectHalfHeight)) (ScalableFloat|GetValueatLevel (Variables|ScalableFloats|GetRow_Homing_Range)) (ScalableFloat|GetValueasBool (Variables|ScalableFloats|GetRow_Homing_TargetAIPawns)) "ECC_GameTraceChannel16" (ScalableFloat|GetValueatLevel (Variables|ScalableFloats|GetRow_Homing_SwitchRectDirectionPitchAngle))))
          (Utilities|IsValid _returnvalue_2
            (:Is%20Valid
              (bind _clearlostotarget (Homing|LOSCheck (Variables|Default|GetWeaponOwner) _returnvalue_2))
              (if _clearlostotarget
                (Class|FortWeaponComponentTargetState|UpdateTargetInfo _fortweaponcomponent_targetstate true (Transformation|GetActorLocation _returnvalue_2))
                (bind _outtargettypegameplaytags (Class|FortWeaponComponentTargetState|DetermineTargetType _fortweaponcomponent_targetstate _returnvalue_2))
                (Class|FortWeaponComponentTargetState|UpdateTargetType _fortweaponcomponent_targetstate _outtargettypegameplaytags)
                (Variables|Homing|SetHomingTarget _returnvalue_2)
                (return)
                (else
                  _)))
            (:Is%20Not%20Valid
              _)))
        (:Is%20Not%20Valid
          _)))
    (:else
      _))
  (Class|FortWeaponComponentTargetState|UpdateTargetInfo :self _fortweaponcomponent_targetstate_2 :bIsTargetValid false)
  (Class|FortWeaponComponentTargetState|UpdateTargetType _fortweaponcomponent_targetstate_2 (GameplayTags|MakeGameplayTagContainerfromTag))
  (Variables|Homing|SetHomingTarget 0)
  (return))

(fn ConstructionScript)
