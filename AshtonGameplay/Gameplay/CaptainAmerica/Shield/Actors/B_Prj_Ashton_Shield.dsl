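; Captain America's thrown shield projectile for the Ashton gameplay: it damages and knocks back pawns, vehicles and buildings, then returns to the thrower.
; Timelines scale speed and spin from ScalableFloat rows; a timer, hit limits or certain impacts trigger the homing return.
; On impacts, the server runs sphere traces to ricochet toward nearby pawns, or mirrors velocity off the surface.
; It holsters the weapon in flight; KillProjectile sends the catch event and starts cleanup.
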(event Custom|StopMovementSpeedScaling
  ;[1 1 1] Scale the movement speed after it is thrown
  (bind (_direction _speedratio) (Timeline|Timeline_MovementSpeedScale Stop)
    (:Update
      (if (not (Variables|BPSetVariables|GetReturnToSender))
        (Variables|/Script/Engine.ProjectileMovementComponent|SetMaxSpeed (Math|Float|Lerp (Variables|BPSetVariables|GetInitialScalingSpeed) (Variables|BPSetVariables|GetSlowScalingSpeed) _speedratio) (Variables|Movement|GetProjectileMovementComponent))))))

(event Custom|ScaleMovementSpeed
  ;[1 1 1] Scale the movement speed after it is thrown
  ;;[1 1 1] Just Get the values once since this is called every frame for a second.
  (Variables|BPSetVariables|SetInitialScalingSpeed (ScalableFloat|GetValueatLevel (Variables|ScalableFloats|GetRow_StartVelocity)))
  (Variables|BPSetVariables|SetSlowScalingSpeed (ScalableFloat|GetValueatLevel (Variables|ScalableFloats|GetRow_EndVelocity)))
  (Components|Timeline|SetPlayRate (Variables|B_Prj_Ashton_Shield|GetTimeline_MovementSpeedScale) (/ 1.0 (ScalableFloat|GetValueatLevel (Variables|ScalableFloats|GetRow_AutoReturnTime))))
  (Timeline|Timeline_MovementSpeedScale PlayFromStart))

(event EventBeginPlay
  ;[1 1 1] EVENT -> Begin Play
  (Utilities|FlowControl|Sequence
    (:then_0
      (Variables|Default|SetSpawnFrame (Utilities|GetFrameCount))
      (bind _returnvalue (Audio|SpawnSoundAttached (Variables|Default|GetCastSound) (Variables|Collision|GetCapsuleComponent)))
      (Variables|Default|SetCastInAirAudio _returnvalue)
      (bind _castinairaudio (Variables|Default|GetCastInAirAudio)
        (:then
          (SoundIndicator|StartSoundIndicatorLooping (Variables|Default|GetFortSoundIndicator) (Transformation|GetActorLocation) (Variables|/Script/Engine.AudioComponent|GetSound _castinairaudio) 1.0 0))
        (:else)))
    (:then_1
      (Game|Components|ProjectileMovement|Interpolation|SetInterpolatedComponent (Variables|Movement|GetProjectileMovementComponent) (Variables|Default|GetScene)) ; Interpolate the scene component so this doesn't affect the rotation of the projectile
    )
    (:then_2
      (Utilities|Map|Clear (Variables|Default|GetRecentlyHit))
      (CallFunction|ScaleMovementSpeed)
      (Variables|BPSetVariables|SetLastHitNonDestroyedBuilding 0) ; Set LastHitNonDestroyed Building to null
      (bind _asfort_player_pawn (Utilities|Casting|CastToFortPlayerPawn (Game|GetInstigator)))
      (Variables|BPSetVariables|SetPlayerPawn _asfort_player_pawn)
      (bind _returnvalue_2 (Utilities|Time|SetTimerbyEvent (EventDispatchers|CreateEvent :!event Custom|NaturalApexReached) (ScalableFloat|GetValueatLevel (Variables|ScalableFloats|GetRow_AutoReturnTime))))
      (Variables|Default|SetNaturalApexReachedTimerHandle _returnvalue_2)
      (bind _returnvalue_3 (Utilities|Time|SetTimerbyEvent (EventDispatchers|CreateEvent :!event Custom|KillProjectile) (ScalableFloat|GetValueatLevel (Variables|ScalableFloats|GetRow_FailsafeMaxLifeTime)))) ; In case something goes wrong, ensure ability ends
      (Variables|BPSetVariables|SetFailsafeTimerHandle _returnvalue_3)
      ;[1 1 1] If player dies, destroy projectile
      (EventDispatchers|BindEventtoOnDied (Variables|BPSetVariables|GetPlayerPawn) (delegate Custom|OnDied))
      (Game|BindEventtoOnDestroyed (EventDispatchers|CreateEvent :!event Custom|ProjectileDestroyed))
      ;[1 1 1] Rotation in air
      (Components|Movement|SetUpdatedComponent (Variables|Default|GetRotatingMovement) (Variables|Default|GetSkelMesh_Weapon))
      (Components|Timeline|SetPlayRate (Variables|B_Prj_Ashton_Shield|GetTimeline_RotationSlowdown) (/ 1.0 (ScalableFloat|GetValueatLevel (Variables|ScalableFloats|GetRow_RotationSlowdownTime))))
      (Transformation|SetRelativeRotation (Variables|Default|GetScene) "0.000000,0.000000,60.000000" false false)
      (bind (_direction _scalerotation _flattenrot) (Timeline|Timeline_RotationSlowdown PlayFromStart)
        (:Update
          (Variables|/Script/Engine.RotatingMovementComponent|SetRotationRate_Roll (Math|Rotator|MakeRotator 0.0 (Math|Float|Lerp (ScalableFloat|GetValueatLevel (Variables|ScalableFloats|GetRow_StartSpinRate)) (ScalableFloat|GetValueatLevel (Variables|ScalableFloats|GetRow_EndSpinRate)) _scalerotation)) (Variables|Default|GetRotatingMovement)) ; decelerate the rotation before it stops moving
          (Transformation|SetRelativeRotation (Variables|Default|GetScene) (Math|Rotator|MakeRotator (Math|Float|Lerp 6e+01 0.0 _flattenrot)) false false))) ; Scale up rotation rate so it isn't rotating instantly when it comes out of hand
    )
    (:then_3
      (bind _playerpawn (Variables|BPSetVariables|GetPlayerPawn)
        (:then
          ;;[0.4577 0.7777 1] Cache the weapon so we can access it later even if it is unequipped
          (bind _asb_ashton_shield_weap (Utilities|Casting|CastToB_Ashton_Shield_Weap (Weapon|GetCurrentWeapon _playerpawn)))
          (Variables|Default|SetOwningWeapon _asb_ashton_shield_weap)
          (Networking|SwitchHasAuthority
            (:Authority
              (bind _owningweapon (Variables|Default|GetOwningWeapon)
                (:then
                  (Class|BAshtonShieldWeap|SetProjectile _owningweapon self))
                (:else)))))
        (:else)))
    (:then_4
      ;;[0.453 1 0.9568] Since the weapon should be holstered while the projectile exists, have the projectile be the holster source and holster the weapon upon its creation
      (bind _playerpawn_2 (Variables|BPSetVariables|GetPlayerPawn)
        (:then
          (Weapon|SetHolsterWeapon _playerpawn_2 0 true false "ashton_shield"))
        (:else)))
    (:then_5
      (Utilities|Array|AddUnique (Variables|Default|GetActorsToIgnoreForTargeting) (Variables|BPSetVariables|GetPlayerPawn))
      (CallFunction|HandleThrowingfromVehicle))))

(event EventEndPlay
  (params (EndPlayReason :type EEndPlayReason%20Enum :default "Destroyed"))
  ;[1 1 1] EVENT -> End Play
  (|CallParentFunction EndPlayReason)
  ;;[1 1 1] Stop audio if gets stuck (fixing issue with bug that audio gets stuck if throwing lobster right at the ground)
  (bind _castinairaudio (Variables|Default|GetCastInAirAudio)
    (:then
      (Audio|Components|Audio|FadeOut _castinairaudio 0.5 0.0 "Linear" :!class /Script/Engine.AudioComponent))
    (:else
      _))
  (bind _recallaudio (Variables|Default|GetRecallAudio)
    (:then
      (Audio|Components|Audio|FadeOut _recallaudio 0.5 0.0 "Linear" :!class /Script/Engine.AudioComponent))
    (:else)))

(event Custom|OnDied
  (params (DamagedActor :type Actor%20Object%20Reference) (Damage :type Float%20%28single-precision%29) (InstigatedBy :type Controller%20Object%20Reference) (DamageCauser :type Actor%20Object%20Reference) (HitLocation :type Vector) (FHitComponent :type Primitive%20Component%20Object%20Reference) (BoneName :type Name) (Momentum :type Vector))
  ;[1 1 1] If player dies, destroy projectile
  (CallFunction|KillProjectile))

(event OnProjectileStop(ProjectileMovementComponent)
  (params (ImpactResult :type Hit%20Result%20Structure%20%28by%20ref%29))
  (bind _hitresult (Variables|Default|GetHitResult))
  (bind (_bblockinghit _binitialoverlap _time _distance _location _impactpoint _normal _impactnormal _physmat _hitactor) (Collision|BreakHitResult _hitresult))
  (bind (_normal_x _normal_y _normal_z) (Math|Vector|BreakVector _normal))
  (bind _returnvalue (Utilities|GetDisplayName (.HitActor (Collision|BreakHitResult ImpactResult))))
  (bind (_bblockinghit_2 _binitialoverlap_2 _time_2 _distance_2 _location_2 _impactpoint_2 _normal_2 _impactnormal_2 _physmat_2 _hitactor_2) (Collision|BreakHitResult ImpactResult))
  (bind _lasthitobstacle (Variables|BPSetVariables|GetLastHitObstacle))
  (bind _lasthitobstacle_2 (Variables|BPSetVariables|GetLastHitObstacle))
  ;[0.6406 1 0.6843] EVENT -> On Projectile Stop
  (Networking|SwitchHasAuthority
    (:Authority
      (Utilities|IsValid _hitactor_2
        (:Is%20Valid
          (bind _continuehitlogic (CallFunction|CheckHitObject _hitactor_2))
          (if _continuehitlogic
            (Utilities|FlowControl|Sequence
              (:then_0
                ; Clear move ignore actors after another hit (bounce) if not returning.
                (if (not (Variables|BPSetVariables|GetIsReturning))
                  (Collision|ClearMoveIgnoreActors (Variables|Collision|GetCapsuleComponent))))
              (:then_1
                (Effects|Components|ParticleSystem|SetBoolParameter (Variables|Default|GetNS_AnimTrail) "Stopped" true)
                (CallFunction|UpdateTrailFXOnImpact))
              (:then_2
                (if (Variables|/Script/Engine.Actor|GetbCanBeDamaged _hitactor_2)
                  (Variables|Default|SetHitResult ImpactResult)
                  (bind _asfort_vehicle_interface (Utilities|Casting|CastToFortVehicleInterface _hitactor_2)
                    (:then
                      ;;[0.1752 1 0.4401] Hit and Damage Vehicles
                      (CallFunction|DamageHitActor (Vehicle|GetActor _asfort_vehicle_interface))
                      (Development|PrintString "hit vehicle" true true "(R=0.000000,G=1.000000,B=0.020782,A=1.000000)" 5.0)
                      (Collision|IgnoreActorwhenMoving (Variables|Collision|GetCapsuleComponent) (Vehicle|GetActor _asfort_vehicle_interface) true)
                      (if (Variables|BPSetVariables|GetReturnToSender)
                        (CallFunction|HandleProjectileStopReturn)
                        (else
                          (CallFunction|IncrementThingsHit)
                          (CallFunction|ShieldSphereTrace true))))
                    (:CastFailed
                      ;;[1 0.4908 0.4609] Hit and Damaged Building
                      (bind _asbuilding_actor (Utilities|Casting|CastToBuildingActor _hitactor_2)
                        (:then
                          (bind _output_get (Variables|BPSetVariables|SetLastHitObstacle _asbuilding_actor))
                          (if (and (or (and (> (Health|GetMaxHealth _output_get :!class /Script/FortniteGame.BuildingActor) 0.0) (<= (Health|GetMaxHealth _output_get :!class /Script/FortniteGame.BuildingActor) 5e+03)) (GameplayTags|HasMatchingGameplayTag _output_get "(TagName=\"Trait.Shield.Damageable\")")) (!= _output_get (Variables|BPSetVariables|GetLastHitNonDestroyedBuilding)))
                            (if (Variables|BPSetVariables|GetReturnToSender)
                              (CallFunction|DamageEnvironmentActor _lasthitobstacle_2)
                              (if (not (<= (Health|GetHealth _lasthitobstacle_2 :!class /Script/FortniteGame.BuildingActor) 0))
                                (Variables|BPSetVariables|SetLastHitNonDestroyedBuilding (Variables|BPSetVariables|GetLastHitObstacle)) ; Save this building so we can check it in ViableHit?\nThis fixes buildings taking multiple hits when the saber is "bouncing" off\nof them or passing through them. 
                                (Collision|IgnoreActorwhenMoving (Variables|Collision|GetCapsuleComponent) (Variables|BPSetVariables|GetLastHitObstacle) true) ; If we didn't destroy the wall, pass through it
                              )
                              (CallFunction|HandleProjectileStopReturn) ; continue returning to player
                              (else
                                (CallFunction|IncrementThingsHit)
                                (CallFunction|ShieldSphereTrace true)
                                (CallFunction|DamageEnvironmentActor _lasthitobstacle)
                                (Variables|BPSetVariables|SetLastHitNonDestroyedBuilding (Variables|BPSetVariables|GetLastHitObstacle)) ; Save this building so we can check it in ViableHit?\nThis fixes buildings taking multiple hits when the saber is "bouncing" off\nof them or passing through them. 
                                (if (<= (Health|GetHealth _lasthitobstacle :!class /Script/FortniteGame.BuildingActor) 0)
                                  (Development|PrintString (Utilities|String|Append "BROKE AND RETURN:   " (Utilities|GetDisplayName (Variables|BPSetVariables|GetLastHitObstacle))))
                                  (else
                                    (CallFunction|RemedyDamageableShieldHit (Variables|BPSetVariables|GetLastHitObstacle))
                                    (Development|PrintString (Utilities|String|Append "Damaged and STOP:   " (Utilities|GetDisplayName (Variables|BPSetVariables|GetLastHitNonDestroyedBuilding))))
                                    (Collision|IgnoreActorwhenMoving (Variables|Collision|GetCapsuleComponent) (Variables|BPSetVariables|GetLastHitObstacle) true)
                                    (Transformation|SetActorLocation (+ (Transformation|GetActorLocation) (* (Math|Vector|GetUnitDirection(Vector) (Transformation|GetActorLocation) (Transformation|GetActorLocation (Variables|BPSetVariables|GetPlayerPawn))) 1e+02)) false true) ; If we hit something solid then we need to move projectile \ntowards player or it will be stuck
                                  ))))
                            (else
                              ;;[1 0.4524 0.2474] Hit a destructible item, but did not destroy it so return to player
                              (Development|PrintString (Utilities|String|Append "ViableHit returned false, ignoring hit on " (Utilities|GetDisplayName (Variables|BPSetVariables|GetLastHitObstacle)))))))
                        (:CastFailed
                          ;;[1 1 0] Hit a SimpleDesctructiblePropGroup Actor
                          (bind _assimple_destructible_prop_group (Utilities|Casting|CastToSimpleDestructiblePropGroup _hitactor_2)
                            (:then
                              (CallFunction|DamageEnvironmentActor _assimple_destructible_prop_group)
                              (if (Variables|BPSetVariables|GetReturnToSender)
                                (CallFunction|HandleProjectileStopReturn) ; continue returning to player
                                (else
                                  (CallFunction|ShieldSphereTrace true)
                                  (CallFunction|IncrementThingsHit))))
                            (:CastFailed
                              (Development|PrintString (Utilities|String|Append "Hit a destructible, non-building actor " _returnvalue))))))
                      (Transformation|SetActorLocation (+ (Transformation|GetActorLocation) (* (Math|Vector|GetUnitDirection(Vector) (Transformation|GetActorLocation) (Transformation|GetActorLocation (Variables|BPSetVariables|GetPlayerPawn))) 1e+02)) false true) ; If we hit something solid then we need to move projectile \ntowards player or it will be stuck
                      (Collision|IgnoreActorwhenMoving (Variables|Collision|GetCapsuleComponent) _hitactor_2 true) ; If we didn't destroy the wall, pass through it
                      (if (GameplayTags|HasMatchingGameplayTag (Variables|BPSetVariables|GetLastHitObstacle) "(TagName=\"BuildingActor.NonDestructable\")")
                        ;;; FORT-766687: For objects with the BuildingActor.NonDestructible tag such as the bubble shield, we return immediately to avoid a bug. Can possibly improve in the future, this means we can't bounce to something after hitting a shield.
                        (CallFunction|GCFeedback (Variables|Default|GetGC_EnvironmentImpact) (Transformation|GetActorLocation self) (Transformation|GetActorForwardVector self))
                        (CallFunction|ReturntoOwner)
                        (else
                          (if (Variables|BPSetVariables|GetReturnToSender)
                            (CallFunction|HandleProjectileStopReturn)
                            (else
                              (CallFunction|ShieldSphereTrace true)
                              (CallFunction|IncrementThingsHit)))))))
                  (else
                    ;;[1 0.6685 0.2208] Hit indestructible
                    ;; If already returning, keep going. If not, return
                    (Development|PrintString (Utilities|String|Append "Hit Indestructible: " _returnvalue))
                    (Collision|IgnoreActorwhenMoving (Variables|Collision|GetCapsuleComponent) _hitactor_2 true)
                    (if (Variables|BPSetVariables|GetReturnToSender)
                      (CallFunction|HandleProjectileStopReturn)
                      (else
                        (Utilities|IsA(soft) _hitactor "/Script/Water.WaterBody"
                          (:True
                            (if (or (WaterBody|IsSurfaceSwimming (Variables|BPSetVariables|GetPlayerPawn)) (GameplayTags|HasMatchingGameplayTag (Variables|BPSetVariables|GetPlayerPawn) "(TagName=\"Curie.State.ElementInteraction.Water\")"))
                              (if (> _normal_z 0)
                                (CallFunction|ResumeMovement true) ; Throwing out of water while swimming
                                (else
                                  _))
                              (elif (>= (- (Utilities|GetFrameCount) (Variables|Default|GetSpawnFrame)) (ScalableFloat|GetValueasInteger (Variables|ScalableFloats|GetRow_MinFramesForIndestructableBounce)))
                                (CallFunction|ShieldSphereTrace false)
                                (CallFunction|IncrementThingsHit)
                                (CallFunction|GCFeedback (Variables|Default|GetGC_EnvironmentImpact) (Transformation|GetActorLocation self) (Transformation|GetActorForwardVector self) _hitactor_2)))
                            (CallFunction|ReturntoOwner) ; Throwing into water while swimming, or throwing into indestructable at point-blank range (hit on spawn)
                          )
                          (:False
                            _))))))))))))))

(event Collision|EventHit
  (params (MyComp :type Primitive%20Component%20Object%20Reference) (Other :type Actor%20Object%20Reference) (OtherComp :type Primitive%20Component%20Object%20Reference) (bSelfMoved :type Boolean) (HitLocation :type Vector :default "0, 0, 0") (HitNormal :type Vector :default "0, 0, 0") (NormalImpulse :type Vector :default "0, 0, 0") (Hit :type Hit%20Result%20Structure%20%28by%20ref%29))
  (|CallParentFunction MyComp Other OtherComp bSelfMoved HitLocation HitNormal NormalImpulse Hit))

(event Custom|ShieldSphereTrace
  (params (CheckReturnEarlyHF :type Boolean))
  (bind _array (Utilities|Array|MakeArray "ObjectTypeQuery3" "ObjectTypeQuery7"))
  ;[1 1 1] Shield Sphere Trace
  (Variables|Default|SetbTargetAcquired false)
  (Variables|Default|SetPawnToTravelTo 0)
  (Variables|Default|SetDistanceToNearestTarget 1e+04)
  (bind _outactors (Collision|SphereOverlapActors (Transformation|GetActorLocation) (ScalableFloat|GetValueatLevel (Variables|Default|GetDataHitSphereTraceRadius)) _array "/Script/FortniteGame.FortPawn" (Variables|Default|GetActorsToIgnoreForTargeting)))
  (for _array_element _outactors
    (bind _asfort_pawn (Utilities|Casting|CastToFortPawn _array_element))
    (Variables|Default|SetLocalOverlapActor _asfort_pawn)
    (CallFunction|CanTraveltoOverlapActor))
  (Variables|Default|GetPawnToTravelTo
    (:then
      _)
    (:else
      ;; FORT-777839: Deimos AI aren't returning as part of Sphere Overlap Actors, so we run a separate multi sphere trace for objects to look for them.
      (if (ScalableFloat|GetValueasBool (Variables|ScalableFloats|GetRow_OnBounceSecondaryMultiSphereCheck))
        (bind _outhits (Collision|MultiSphereTraceForObjects (Transformation|GetActorLocation) (Transformation|GetActorLocation) (ScalableFloat|GetValueatLevel (Variables|Default|GetDataHitSphereTraceRadius)) _array false (Variables|Default|GetActorsToIgnoreForTargeting)))
        (for _array_element_2 _outhits
          (bind _asfort_pawn_2 (Utilities|Casting|CastToFortPawn (.HitActor (Collision|BreakHitResult _array_element_2))))
          (Variables|Default|SetLocalOverlapActor _asfort_pawn_2)
          (CallFunction|CanTraveltoOverlapActor))
        (else
          _))
      (bind _output_get (Variables|Default|SetShieldTarget (Variables|Default|GetPawnToTravelTo)))
      (Utilities|IsValid _output_get
        (:Is%20Valid
          (Variables|Default|SetbTargetAcquired true)
          (bind _ascharacter (Utilities|Casting|CastToCharacter (Variables|Default|GetShieldTarget)))
          (bind _output_get_2 (Variables|Default|SetHomingTarget _ascharacter))
          (Default|SetHomingVars true)
          (Variables|/AshtonGameplay/Gameplay/CaptainAmerica/Shield/Actors/B_Ashton_Shield_Weap.B_Ashton_Shield_Weap_C|SetHomingTarget _output_get_2 (Variables|Default|GetOwningWeapon))
          (CallFunction|HandleProjectileStopReturn))
        (:Is%20Not%20Valid
          (Variables|Default|SetVector_AngleOfIncidence (Math|Vector|MirrorVectorbyNormal (Variables|BPSetVariables|GetPreviousVelocity) (Variables|Default|GetShieldHitNormal)))
          (if (or (> (Utilities|Map|Length (Variables|Default|GetRecentlyHit)) 0) (and (ScalableFloat|GetValueasBool (Variables|ScalableFloats|GetROW_ReturnEarlyHF_Enabled)) CheckReturnEarlyHF))
            (CallFunction|ReturntoOwner)
            (CallFunction|HandleProjectileStopReturn)
            (else
              (Variables|/Script/Engine.MovementComponent|SetVelocity (Variables|Default|GetVector_AngleOfIncidence) (Variables|Movement|GetProjectileMovementComponent))
              (CallFunction|HandleProjectileStopReturn))))))))

(event Custom|NaturalApexReached
  ;[1 1 1] Return To Owner
  (Development|PrintString "Reached throw Apex")
  (CallFunction|ReturntoOwner))

(event Custom|CallReturnMovement
  ;[1 1 1] Call Return Movement
  (CallFunction|ResumeMovement false) ; Do not reset the velocity when returning
  (CallFunction|PlayImpactFeedback))

(event Custom|AnimTrailNotifyBegin
  (params (bActive :type Boolean))
  ;[1 1 1] Niagara Anim Trail Activation
  (Niagara|SetNiagaraVariable(Bool) (Variables|Default|GetNS_AnimTrail) "AnimTrailActivate" bActive))

(event Custom|PlayImpactFeedback
  ;[1 1 1] Trigger controller feedback
  (Networking|CanExecuteCosmeticEvents
    (:True
      (bind _playerpawn (Variables|BPSetVariables|GetPlayerPawn)
        (:then
          (if (Pawn|IsLocallyControlled _playerpawn)
            (ForceFeedback|SpawnForceFeedbackAttached "/Game/Weapons/ForceFeedback/FF_Generic_Impact_Near.FF_Generic_Impact_Near" (Variables|/Script/Engine.Actor|GetRootComponent _playerpawn))))
        (:else)))))

(event Custom|ReturnToOwner
  (bind _playerpawn (Variables|BPSetVariables|GetPlayerPawn))
  ;[1 1 1] Return To Owner
  (Networking|SwitchHasAuthority
    (:Authority
      (Development|PrintString "Return to Owner Triggered")
      (CallFunction|HandleProjectileStopReturn)
      (Utilities|IsValid _playerpawn
        (:Is%20Valid
          (bind _distance (Class|MissionBlueprintFunctionLibrary|DistanceBetweenTwoVectors (Transformation|GetActorLocation) (Transformation|GetActorLocation _playerpawn)))
          (if (>= _distance (ScalableFloat|GetValueatLevel (Variables|ScalableFloats|GetRow_ReturnHomingLockTargetDistanceThreshold)))
            (Variables|BPSetVariables|SetIsReturning true)
            ;;[1 0.4686 0] Delay a tick before setting new target and ReturnToSender. Prevents OnReachedLockTargetDistanceThreshold from being called instantly with ReturnToSender set.
            (Utilities|FlowControl|DelayUntilNextTick)
            (CallFunction|TurnonRecallFX)
            (Variables|BPSetVariables|SetReturnToSender true) ; Set Return Things
            (Networking|ForceNetUpdate)
            (else
              _)))
        (:Is%20Not%20Valid
          _))
      (CallFunction|KillProjectile))))

(event Custom|CallMovement
  ;[1 1 1] Call Outgoing Movement
  (CallFunction|ResumeMovement true)
  (CallFunction|PlayImpactFeedback))

(event Custom|AnimTrailNotifyEnd
  ;[1 1 1] Niagara Anim Trail Activation
  (Niagara|SetNiagaraVariable(Bool) (Variables|Default|GetNS_AnimTrail) "AnimTrailActivate" false))

(event Custom|TurnOnRecallFX
  ;[1 1 1] Turn on return FX and sounds
  (Utilities|FlowControl|DoOnce false
    (:Completed
      (Development|PrintString "Recall FX")
      (CallFunction|PlayRecallSound)
      (CallFunction|AnimTrailNotifyEnd)
      (Utilities|FlowControl|Sequence
        (:then_0
          ;[1 1 1] Start rotating again & rotate projectile back to hand catch orientation when close enough
          (bind (_direction _speed) (Timeline|Timeline_RotateOnReturn PlayFromStart)
            (:Update
              (Variables|/Script/Engine.RotatingMovementComponent|SetRotationRate_Roll (Math|Rotator|MakeRotator 0.0 (Math|Float|Lerp (ScalableFloat|GetValueatLevel (Variables|ScalableFloats|GetRow_EndSpinRate)) (ScalableFloat|GetValueatLevel (Variables|ScalableFloats|GetRow_StartSpinRate)) _speed)) (Variables|Default|GetRotatingMovement)))))
        (:then_1
          (if (CallFunction|IsCloserThanReturnRotateThreshold)
            (CallFunction|ShouldRotateShieldforReturn)
            (else
              (bind _returnvalue (Utilities|Time|SetTimerbyFunctionName 0 "ShouldRotateShieldForReturn" 0.1 true))
              (Variables|Default|SetTimer_CheckRotateOnReturn _returnvalue))))))))

(event EventOnTouched
  (params (OtherActor :type Actor%20Object%20Reference) (OtherComp :type Primitive%20Component%20Object%20Reference) (HitResult :type Hit%20Result%20Structure%20%28by%20ref%29) (bIsOverlap :type Boolean))
  (bind (_bblockinghit _binitialoverlap _time _distance _location _impactpoint _normal) (Collision|BreakHitResult HitResult))
  ;[0.4662 1 0.9235] Event OnTouched - If overlap, call HandleOverlap to potentially damage a character
  ; If not an overlap, cache the velocity for use in the on stop event
  (Variables|Default|SetShieldHitLoc _location)
  (Variables|Default|SetShieldHitNormal _normal)
  (if bIsOverlap
    (Variables|Default|SetHitResult HitResult)
    (Variables|Default|SetLimitBounceAngle false)
    (CallFunction|HandleOverlap OtherActor)
    (else
      (Variables|BPSetVariables|SetPreviousVelocity (Transformation|GetVelocity)) ; Set Previous Velocity so on ResumeMovement we know the direction to go.\nSetting it here because this gets triggered before the stop event
      (Variables|Default|SetLimitBounceAngle true))))

(event Custom|RotateShieldForReturn
  ;[1 1 1] Rotate Shield For Return
  (Components|Timeline|SetPlayRate (Variables|B_Prj_Ashton_Shield|GetTimeline_ReturnToHand) (ScalableFloat|GetValueatLevel (Variables|ScalableFloats|GetRow_ReturnRotationTimelineRate)))
  (bind (_direction _rotation) (Timeline|Timeline_ReturnToHand PlayFromStart)
    (:Update
      (Transformation|SetWorldRotation (Variables|Default|GetSkelMesh_Weapon) (Math|Rotator|Lerp(Rotator) (Variables|Default|GetcachedMeshRotation) (Math|Rotator|FindLookatRotation (Transformation|GetActorLocation) (Transformation|GetSocketLocation (Variables|/Script/Engine.Character|GetMesh (Variables|BPSetVariables|GetPlayerPawn)) "hand_attach_r")) _rotation) false false))))

(event Custom|DamageEnvironmentActor
  (params (Actor :type Actor%20Object%20Reference))
  (bind _returnvalue (Ability|GetAbilitySystemComponent Actor))
  ;[1 1 1] Damage building/environment actor
  (CallFunction|GCFeedback (Variables|Default|GetGC_EnvironmentImpact) (Transformation|GetActorLocation self) (Transformation|GetActorForwardVector self) Actor)
  (bind _bdidactorhavetags (Class|AthenaQuestFunctionLibrary|AthenaCheckActorHasTags Actor (Variables|Default|GetInstaKillTags)))
  (Utilities|IsValid _returnvalue
    (:Is%20Valid
      (CallFunction|AddHitResulttoEffectContext)
      (Utilities|Casting|CastToBGA_Ashton_Shield_BlockVolume Actor
        (:then
          (GameplayEffects|ApplyGameplayEffectToTarget (Ability|GetAbilitySystemComponent (Variables|BPSetVariables|GetPlayerPawn)) "/AshtonGameplay/Gameplay/CaptainAmerica/Shield/Effects/GE_Ashton_Shield_Throw_Damage_Player.GE_Ashton_Shield_Throw_Damage_Player_C" _returnvalue 0.0 (Variables|Default|GetHitEffectContext)) ; Blocking shields take player damage, not environment damage
        )
        (:CastFailed
          (GameplayEffects|ApplyGameplayEffectToTarget (Ability|GetAbilitySystemComponent (Variables|BPSetVariables|GetPlayerPawn)) (select _bdidactorhavetags (Variables|Default|GetProjectileEnvDamage_InstaKill) (Variables|Default|GetProjectileEnvDamage)) _returnvalue 0.0 (Variables|Default|GetHitEffectContext))))
      (Utilities|Map|Add (Variables|Default|GetRecentlyHit) Actor (Fortnite|Sync|Time|GetServerWorldTimeSeconds)))))

(event Custom|HandleOverlap
  (params (OverlapActor :type Actor%20Object%20Reference))
  (bind _row_minimumspeednoautoreturn (Variables|ScalableFloats|GetROW_MinimumSpeedNoAutoReturn))
  (bind _returnvalue (ScalableFloat|GetValueatLevel _row_minimumspeednoautoreturn))
  ;[0.5948 0.9165 1] Overlap - Potentially damage character
  (Networking|SwitchHasAuthority
    (:Authority
      (bind _ascharacter (Utilities|Casting|CastToCharacter OverlapActor))
      ;;[1 1 1] ignore the hit if player was recently hit
      (bind _isimmune (CallFunction|CheckImmunityStatus _ascharacter))
      (if (not _isimmune)
        (if (not (Math|Boolean|ORBoolean (Utilities|Equal(Object) (Variables|Actor|GetInstigator) _ascharacter) (Team|OnSameTeam _ascharacter (Variables|BPSetVariables|GetPlayerPawn)) (Utilities|Array|ContainsItem (Actor|GetAttachedActors _ascharacter true false) (Variables|Actor|GetInstigator))))
          (CallFunction|EvaluateEndingHomingBehavior _ascharacter) ; stop homing if we reached the target
          (CallFunction|IncrementThingsHit)
          (CallFunction|DamageHitActor _ascharacter)
          (Utilities|Map|Add (Variables|Default|GetRecentlyHit) _ascharacter (Fortnite|Sync|Time|GetServerWorldTimeSeconds))
          (Utilities|Array|AddUnique (Variables|Default|GetActorsToIgnoreForTargeting) _ascharacter)
          (if (ScalableFloat|GetValueasBool (Variables|ScalableFloats|GetROW_CancelAutoReturnOnPawnHit))
            ;; HandleCancelAutoReturnOnPawnHit
            (Utilities|Time|ClearandInvalidateTimerbyHandle (Variables|Default|GetNaturalApexReachedTimerHandle))
            (CallFunction|StopMovementSpeedScaling)
            (if (<= (Variables|/Script/Engine.ProjectileMovementComponent|GetMaxSpeed (Variables|Movement|GetProjectileMovementComponent)) _returnvalue)
              (Variables|/Script/Engine.ProjectileMovementComponent|SetMaxSpeed _returnvalue (Variables|Movement|GetProjectileMovementComponent))
              (else
                _))
            (else
              _))
          (bind _output_get (Variables|Default|SetTotalPawnsDamaged (+ (Variables|Default|GetTotalPawnsDamaged) 1)))
          (if (>= _output_get (ScalableFloat|GetValueasInteger (Variables|ScalableFloats|GetROW_MaxPawnsDamaged)))
            (CallFunction|ReturntoOwner)
            (else
              (CallFunction|ShieldSphereTrace false))))
        (else
          (Development|PrintString "Immune from Attack, ignoring"))))))

(event Custom|DamageHitActor
  (params (Actor :type Actor%20Object%20Reference))
  (bind (_case_0_true _case_1_true) (GameplayTags|TagContainer|CompareTagAssetInterfacetoOtherTags (Variables|Default|GetInVehicle) (Variables|Default|GetInVehicle_SeatOpen) (Ability|GetAbilitySystemComponent Actor)))
  (bind _returnvalue (Game|GetInstigator))
  (bind _returnvalue_2 (Ability|GetAbilitySystemComponent _returnvalue))
  (bind _returnvalue_3 (Ability|GetAbilitySystemComponent Actor))
  ;[0.4688 0.3006 0.1296] Damage Hit Actor
  (Utilities|IsValid Actor
    (:Is%20Valid
      (bind _owningweapon (Variables|Default|GetOwningWeapon)
        (:then
          (Class|BAshtonShieldWeap|HandleProjectileHit _owningweapon))
        (:else
          _))
      (CallFunction|GCFeedback (Variables|Default|GetGC_PlayerImpact) (Transformation|GetActorLocation self) (Transformation|GetActorForwardVector self) Actor)
      (CallFunction|LaunchHitActor Actor)
      (Utilities|IsValid _returnvalue_3
        (:Is%20Valid
          (Utilities|IsValid _returnvalue
            (:Is%20Valid
              (Utilities|IsValid _returnvalue_2
                (:Is%20Valid
                  (if (or (not _case_0_true) (and _case_0_true _case_1_true))
                    (CallFunction|AddHitResulttoEffectContext)
                    (CallFunction|ApplyPawnDamage _returnvalue_2 _returnvalue_3)))))))))))

(event Custom|LaunchHitActor
  (params (Actor :type Actor%20Object%20Reference))
  (bind _returnvalue (CallFunction|CalculateKnockbackVelocity Actor))
  ;[1 1 1] Launch Actor
  (if (ScalableFloat|GetValueasBool (Variables|ScalableFloats|GetROW_KnockbackEnabled))
    (bind _ascharacter (Utilities|Casting|CastToCharacter Actor)
      (:then
        (bind _returnvalue_2 (GameplayTags|HasMatchingGameplayTag(Message) _ascharacter "(TagName=\"Gameplay.Status.KnockbackImmunity\")"))
        (if (not _returnvalue_2)
          (Character|LaunchCharacter _ascharacter _returnvalue true true)))
      (:CastFailed
        (bind _asfort_vehicle_interface (Utilities|Casting|CastToFortVehicleInterface Actor))
        (bind _returnvalue_3 (Vehicle|GetAllPassengers(Message) (Utilities|ToObject(Interface) _asfort_vehicle_interface)))
        (if (not (Utilities|Array|ContainsItem _returnvalue_3 (Variables|BPSetVariables|GetPlayerPawn)))
          (Audio|BroadcastAbilityImpactatLocation 0 (Utilities|ToObject(Interface) _asfort_vehicle_interface) _returnvalue 777.0) ; This launches vehicles (named weird for anti-cheat reasons)\nPosition = velocity to launch it at\nRadius must equal 777\nGates knows more about it
        )))))

(event Custom|GCFeedback
  (params (GameplayCueTag :type Gameplay%20Tag%20Structure) (Location :type Vector) (Normal :type Vector) (SourceObject :type Object%20Reference))
  ;[1 0.8018 0] GAMEPLAY CUE FEEDBACK
  (Utilities|FlowControl|DelayUntilNextTick) ; Adding so client gets correct audio multihit value due to order of operations
  (GameplayCue|ExecuteGameplayCueOnActor(Burst) self GameplayCueTag (Ability|GameplayCue|MakeGameplayCueParameters :NormalizedMagnitude (Math|Conversions|ToFloat(Boolean) (Variables|Default|GetbIsAudioMultiHit)) :RawMagnitude 0.0 :Location Location :Normal Normal :Instigator self :EffectCauser self :SourceObject SourceObject :GameplayEffectLevel 1 :AbilityLevel 1 :bReplicateLocationWhenUsingMinimalRepProxy false)))

(event Custom|UpdateTrailFXOnImpact
  (bind _ns_animtrail (Variables|Default|GetNS_AnimTrail))
  ;[0.09232 0.2926 0.325 0.5] Impact FX reaction
  (Niagara|SetNiagaraVariable(Bool) _ns_animtrail "Shield_Impact" true)
  (Utilities|FlowControl|Delay 0.17)
  (Niagara|SetNiagaraVariable(Bool) _ns_animtrail "Shield_Impact" false))

(event Custom|SpawnFireAudio
  ; Triggered through GCN_Ashton_Shield_Throw_Launch_SFX, spawning audio component here so we can interrupt it upon impact
  (bind _returnvalue (Audio|SpawnSoundAttached (Variables|Sounds|GetFireAudio) (Variables|Collision|GetCapsuleComponent)))
  (Variables|Default|SetFireAudioComponent _returnvalue))

(event ProjectileMovement|EventOnReachedLockTargetDistanceThreshold
  ;[1 1 1] EVENT -> Reached Homing Target
  (if (Variables|BPSetVariables|GetReturnToSender)
    (Utilities|FlowControl|DelayUntilNextTick) ; This is called within the projectile component \ntick while loop, so stopping simulation \nthis frame will cause an access crash
    (CallFunction|KillProjectile)))

(event ProjectileMovement|EventOnHomingPassedTarget
  ;[1 1 1] EVENT -> Reached Homing Target
  (if (Variables|BPSetVariables|GetReturnToSender)
    (Projectile|SetHomingTurnSpeedMax 0 0 5e+04 :!class /Script/FortniteGame.FortProjectileMovementInterface)))

(event Custom|KillProjectile
  (bind _projectilemovementcomponent (Variables|Movement|GetProjectileMovementComponent))
  (bind _playerpawn (Variables|BPSetVariables|GetPlayerPawn))
  ;[1 0.7007 0.4203] Kill Projectile (Called from Pawn's OnDied and when it reaches its homing threshold)
  (Utilities|FlowControl|DoOnce false
    (:Completed
      (Utilities|FlowControl|Sequence
        (:then_0
          (Components|Movement|StopMovementImmediately _projectilemovementcomponent)
          (Game|Components|ProjectileMovement|StopSimulating _projectilemovementcomponent (Collision|MakeHitResult false false 0.0 0.0 (Transformation|GetActorLocation)))
          (Rendering|SetVisibility (Variables|Default|GetScene) false true)
          (Development|PrintString "Stopped and Hid Projectile"))
        (:then_1
          (Utilities|IsValid _playerpawn
            (:Is%20Valid
              (Utilities|IsValid (Ability|GetAbilitySystemComponent _playerpawn)
                (:Is%20Valid
                  (Ability|SendGameplayEventtoActor (Variables|BPSetVariables|GetPlayerPawn) (Variables|BPSetVariables|GetEvent_Catch)) ; Trigger Catch GA
                )
                (:Is%20Not%20Valid
                  _))
              (SoundIndicator|StopSoundIndicatorLooping (Variables|Default|GetFortSoundIndicator))
              (bind _fireaudiocomponent (Variables|Default|GetFireAudioComponent)
                (:then
                  (Audio|Components|Audio|FadeOut _fireaudiocomponent 0.5 0.0 "Linear" :!class /Script/Engine.AudioComponent))
                (:else
                  _))
              (bind _castinairaudio (Variables|Default|GetCastInAirAudio)
                (:then
                  (Audio|Components|Audio|FadeOut _castinairaudio 0.5 0.0 "Linear" :!class /Script/Engine.AudioComponent))
                (:else
                  _))
              (bind _recallaudio (Variables|Default|GetRecallAudio)
                (:then
                  (Audio|Components|Audio|FadeOut _recallaudio 0.35 0.0 "Linear" :!class /Script/Engine.AudioComponent))
                (:else
                  _))
              (Networking|SwitchHasAuthority
                (:Authority
                  (bind _owningweapon (Variables|Default|GetOwningWeapon)
                    (:then
                      (Class|BAshtonShieldWeap|SetProjectile _owningweapon 0))
                    (:else)))))
            (:Is%20Not%20Valid
              _)))
        (:then_2
          (Utilities|IsValid (Variables|Default|GetOwningWeapon)
            (:Is%20Valid
              (Class|BAshtonShieldWeap|ClearLockTargets (Variables|Default|GetOwningWeapon)))))
        (:then_3
          ;;[1 1 1] Call ProjectileDestroyed directly so we can start processing cleanup. However, call Kill instead of DestroyActor to set the lifespan to a small duration. This gives time for the server to finish replicating
          (CallFunction|ProjectileDestroyed self)
          (Networking|SwitchHasAuthority
            (:Authority
              (Delivery|Kill))))))))

(event Custom|ProjectileDestroyed
  (params (DestroyedActor :type Actor%20Object%20Reference))
  (bind _playerpawn (Variables|BPSetVariables|GetPlayerPawn))
  (bind _returnvalue (Ability|GetAbilitySystemComponent _playerpawn))
  (bind _playerpawn_2 (Variables|BPSetVariables|GetPlayerPawn))
  (bind _returnvalue_2 (Weapon|GetCurrentWeapon _playerpawn_2))
  ;[1 0.3227 0.1441] ProjectileDestroyed (called from KillProjectile directly and also bound to OnDestroyed)
  (Utilities|FlowControl|DoOnce false
    (:Completed
      (Development|PrintString "ProjectileDestroyed")
      (Utilities|IsValid _playerpawn
        (:Is%20Valid
          (Utilities|IsValid _returnvalue
            (:Is%20Valid
              (GameplayEffects|RemoveActiveGameplayEffect _returnvalue (Variables|BPSetVariables|GetQuestCheckTag_GE) -1)
              ;;[0.9734 0.5271 0.1912] Remove GE_Realm_ProjectileTracker and "ashton_shield" holster ID. This was originally handled in the catch GA, but moved to here as the catch GA can be removed from the player if they switch weapons.
              (GameplayEffects|RemoveActiveEffectswithGrantedTags _returnvalue "(GameplayTags=((TagName=\"Granted.Ashton.Shield.Unusable\")))")
              (Weapon|SetHolsterWeapon _playerpawn_2 0 false false "ashton_shield")
              (Networking|SwitchHasAuthority
                (:Authority
                  (bind _asb_ashton_shield_weap (Utilities|Casting|CastToB_Ashton_Shield_Weap _returnvalue_2))
                  (if (FortWeapon|IsTargeting _returnvalue_2)
                    (bind _blockingbga (Variables|/AshtonGameplay/Gameplay/CaptainAmerica/Shield/Actors/B_Ashton_Shield_Weap.B_Ashton_Shield_Weap_C|GetBlockingBGA _asb_ashton_shield_weap)
                      (:then
                        (Class|BGAAshtonShieldBlockVolume|ToggleBlockProjectiles _blockingbga true)
                        (Variables|/AshtonGameplay/Gameplay/CaptainAmerica/Shield/Actors/BGA_Ashton_Shield_BlockVolume.BGA_Ashton_Shield_BlockVolume_C|SetbActiveBlocking true _blockingbga))
                      (:else))))))))))))

(fn ACloserThanB?
  (params (A :type Actor%20Object%20Reference) (B :type Actor%20Object%20Reference))
  (return (< (Class|AbilityFunctions|GetDistanceBetweenActors A (Actor|GetOwner)) (Class|AbilityFunctions|GetDistanceBetweenActors B (Actor|GetOwner)))))

(fn AddHitResultToEffectContext
  (bind _returnvalue (Game|GetInstigator))
  (bind _returnvalue_2 (Ability|GetFortAbilitySystemComponent _returnvalue))
  (Utilities|IsValid _returnvalue_2
    (:Is%20Valid
      (bind _output_get (Variables|Default|SetHitEffectContext (GameplayEffects|MakeFortEffectContext _returnvalue_2 (Variables|Default|GetOwningWeapon) (Variables|Default|GetOwningWeapon) self 0)))
      (Ability|EffectContext|AddHitResult _output_get (Variables|Default|GetHitResult)))))

(fn ApplyPawnDamage
  (params (InstigatorASC :type Ability%20System%20Component%20Object%20Reference) (HitActorASC :type Ability%20System%20Component%20Object%20Reference))
  (bind _result (ArrayMacros|SafeArrayGet (LocalVariables|GetGEDamageBasedOnPawnsHit) (Variables|Default|GetTotalPawnsDamaged))
    (:%20%20
      (GameplayEffects|ApplyGameplayEffectToTarget (Variables|GetInstigatorASC) _result (Variables|GetHitActorASC) 0.0 (Variables|Default|GetHitEffectContext)))))

(fn CalculateKnockbackVelocity
  (params (Actor :type Actor%20Object%20Reference))
  (bind (_ascharacter _bsuccess) (Utilities|Casting|CastToCharacter Actor))
  (return (Math|Vector|MakeVector (.x (* (Math|Vector|Normalize (Math|Vector|GetUnitDirection(Vector) (Transformation|GetActorLocation) (Transformation|GetActorLocation Actor))) (Math|Float|SelectFloat (ScalableFloat|GetValueatLevel (Variables|ScalableFloats|GetRow_PlayerKnockbackStrength)) (ScalableFloat|GetValueatLevel (Variables|ScalableFloats|GetRow_VehicleKnockbackStrength)) _bsuccess))) (.y (* (Math|Vector|Normalize (Math|Vector|GetUnitDirection(Vector) (Transformation|GetActorLocation) (Transformation|GetActorLocation Actor))) (Math|Float|SelectFloat (ScalableFloat|GetValueatLevel (Variables|ScalableFloats|GetRow_PlayerKnockbackStrength)) (ScalableFloat|GetValueatLevel (Variables|ScalableFloats|GetRow_VehicleKnockbackStrength)) _bsuccess))) (+ (.z (* (Math|Vector|Normalize (Math|Vector|GetUnitDirection(Vector) (Transformation|GetActorLocation) (Transformation|GetActorLocation Actor))) (Math|Float|SelectFloat (ScalableFloat|GetValueatLevel (Variables|ScalableFloats|GetRow_PlayerKnockbackStrength)) (ScalableFloat|GetValueatLevel (Variables|ScalableFloats|GetRow_VehicleKnockbackStrength)) _bsuccess))) (Math|Float|SelectFloat (ScalableFloat|GetValueatLevel (Variables|ScalableFloats|GetRow_PlayerKnockbackStrength)) (ScalableFloat|GetValueatLevel (Variables|ScalableFloats|GetRow_VehicleKnockbackStrength)) _bsuccess)))))

(fn CanApplyHolster
  (params (HolsterId :type Name%20%28by%20ref%29 :default "None"))
  (return true))

(fn CanBounceToTarget
  (if (and (not (ScalableFloat|GetValueasBool (Variables|ScalableFloats|GetRow_CanBounceTargetDBNO))) (GameplayTags|HasMatchingGameplayTag (Variables|Default|GetLocalOverlapActor) "(TagName=\"Gameplay.Status.DBNO\")"))
    (return false)
    (elif (Variables|Default|GetLimitBounceAngle)
      (return (<= (Math|Trig|Acos(Degrees) (Math|Vector|DotProduct (Math|Vector|Normalize (- (Transformation|GetActorLocation (Variables|Default|GetLocalOverlapActor)) (Transformation|GetActorLocation))) (Math|Vector|Normalize (Variables|BPSetVariables|GetPreviousVelocity)))) (ScalableFloat|GetValueatLevel (Variables|ScalableFloats|GetROW_Bounce_LimitedAngle))))
      (else
        (return true)))))

(fn CanTravelToOverlapActor
  (if (Team|OnSameTeam (Variables|Default|GetLocalOverlapActor) (Variables|BPSetVariables|GetPlayerPawn))
    (Utilities|Array|AddUnique (Variables|Default|GetActorsToIgnoreForTargeting) (Variables|Default|GetLocalOverlapActor)) ; We found a nearby teammate. Ignore him for targeting and LOS checks.
    (else
      (bind _clearlostotarget (CallFunction|LOSCheck (Variables|Default|GetLocalOverlapActor))) ; find out if you can see them
      (if _clearlostotarget
        (bind _validtarget (CallFunction|CanBouncetoTarget))
        (if (and _validtarget (< (Transformation|GetDistanceTo (Variables|Default|GetLocalOverlapActor)) (Variables|Default|GetDistanceToNearestTarget)))
          (Variables|Default|SetPawnToTravelTo (Variables|Default|GetLocalOverlapActor)) ; We have a new closest pawn.
          (Variables|Default|SetDistanceToNearestTarget (Transformation|GetDistanceTo (Variables|Default|GetLocalOverlapActor))) ; We have a new closest distance here.
        )))))

(fn CheckHitObject
  (params (HitActor :type Actor%20Object%20Reference))
  (Utilities|FlowControl|Sequence
    (:then_0
      ;[1 1 1] TODO: Clean this up? Better way to handle this interaction? We're hitting the passive collision volume when a player simply holds the shield to their side even though it doesn't look like it. Ignorng it for now this way.
      (bind _asbga_ashton_shield_block_volume (Utilities|Casting|CastToBGA_Ashton_Shield_BlockVolume (Variables|Default|GetHitActor)))
      (if (or (Variables|/AshtonGameplay/Gameplay/CaptainAmerica/Shield/Actors/BGA_Ashton_Shield_BlockVolume.BGA_Ashton_Shield_BlockVolume_C|GetbActiveBlocking _asbga_ashton_shield_block_volume) (GameplayTags|HasMatchingGameplayTag (Variables|/AshtonGameplay/Gameplay/CaptainAmerica/Shield/Actors/BGA_Ashton_Shield_BlockVolume.BGA_Ashton_Shield_BlockVolume_C|GetOwningPlayer _asbga_ashton_shield_block_volume) "(TagName=\"Gameplay.Action.Ashton.Shield.Charge\")"))
        ; FORT-766814: Ignore the player who blocked the shield with the throw & add to recenltly hit early to avoid the bug.
        (Utilities|Array|AddUnique (Variables|Default|GetActorsToIgnoreForTargeting) (Variables|/AshtonGameplay/Gameplay/CaptainAmerica/Shield/Actors/BGA_Ashton_Shield_BlockVolume.BGA_Ashton_Shield_BlockVolume_C|GetOwningPlayer _asbga_ashton_shield_block_volume))
        (Utilities|Map|Add (Variables|Default|GetRecentlyHit) (Variables|Default|GetHitActor) (Fortnite|Sync|Time|GetServerWorldTimeSeconds))
        (return true)
        (else
          (Collision|IgnoreActorwhenMoving (Variables|Collision|GetCapsuleComponent) (Variables|Default|GetHitActor) true) ; Pass through the blocking volume if we're not actively blocking
          (CallFunction|ResumeMovement false)
          (return false))))
    (:then_1
      (return true))))

(fn CheckImmunityStatus
  (params (Key :type Character%20Object%20Reference%20%28by%20ref%29))
  (bind _recentlyhit (Variables|Default|GetRecentlyHit))
  (bind (_value _returnvalue) (Utilities|Map|Find _recentlyhit Key))
  (if _returnvalue
    (if (> (- (Fortnite|Sync|Time|GetServerWorldTimeSeconds) _value) (ScalableFloat|GetValueatLevel (Variables|ScalableFloats|GetRow_PostHitImmunityDuration)))
      (Development|PrintString "Immunity expired, removing from map")
      (Utilities|Map|Remove (Variables|Default|GetRecentlyHit) Key)
      (return false)
      (else
        (return true)))
    (else
      (return false))))

(fn ChooseHomingTargetOffset
  (params (TargetActor :type Character%20Object%20Reference))
  (Utilities|IsValid TargetActor
    (:Is%20Valid
      (bind _returnvalue (GameplayTags|HasMatchingGameplayTag(Message) (Variables|GetTargetActor) "(TagName=\"Pawn.Athena.NPC.Wildlife\")"))
      (if _returnvalue
        ;[0.7898 1 0.48] Wildlife
        (return (Variables|HomingValues|GetWildlifeHomingTargetOffset))
        (else
          (bind _asfort_player_pawn (Utilities|Casting|CastToFortPlayerPawn (Variables|GetTargetActor))
            (:then
              (if (Health|IsDBNO _asfort_player_pawn)
                ;[1 0.593 0.3247] DBNO
                (return (Variables|HomingValues|GetDBNOHomingTargetOffset))
                (else
                  _)))
            (:CastFailed
              _))
          ;[0.5832 0.9901 1] Standard
          (return (Variables|HomingValues|GetHomingTargetOffset)))))
    (:Is%20Not%20Valid
      (return "0, 0, 0"))))

(fn EvaluateEndingHomingBehavior
  (params (HitPawn :type Character%20Object%20Reference))
  (bind (_homingstyle _minturnspeed _maxturnspeed _ramptimefrommintomaxturnspeed _locktargetdistancethreshold _randomseed _lockedontarget) (Utilities|Struct|BreakProjectileHomingData (Variables|Default|GetProjectile%20Homing%20Data)))
  (bind _row_homingeffecttriggerrange (Variables|ScalableFloats|GetRow_HomingEffectTriggerRange))
  (bind _returnvalue (ScalableFloat|GetValueatLevel _row_homingeffecttriggerrange))
  (bind _row_homingminturnspeed (Variables|ScalableFloats|GetRow_HomingMinTurnSpeed))
  (bind _returnvalue_2 (ScalableFloat|GetValueatLevel _row_homingminturnspeed))
  (bind _row_homingmaxturnspeed (Variables|ScalableFloats|GetRow_HomingMaxTurnSpeed))
  (bind _returnvalue_3 (ScalableFloat|GetValueatLevel _row_homingmaxturnspeed))
  (bind _row_homingturnramptime (Variables|ScalableFloats|GetRow_HomingTurnRampTime))
  (bind _returnvalue_4 (ScalableFloat|GetValueatLevel _row_homingturnramptime))
  (bind _projectilehomingdata (Utilities|Struct|MakeProjectileHomingData "None" _returnvalue_2 _returnvalue_3 _returnvalue_4 _returnvalue))
  (bind _projectilemovementcomponent (Variables|Movement|GetProjectileMovementComponent))
  ;[1 1 1] stop chasing the target if we've hit it
  (if (== HitPawn _lockedontarget)
    (Projectile|SetHomingData _projectilemovementcomponent _projectilehomingdata :!class /Script/FortniteGame.FortProjectileMovementComponent) ; reset the homing parameters to stop following the target
    (Variables|Default|SetProjectile%20Homing%20Data _projectilehomingdata)
    (Components|Movement|SetUpdatedComponent _projectilemovementcomponent (Variables|Collision|GetCapsuleComponent))))

(fn FindHomingTarget
  (bind (_outlocation _outrotation) (Actor|GetActorEyesViewPoint (Actor|GetOwner)))
  ;[1 1 1] Shape to collect potential homing targets
  (bind _outactors (Collision|BoxOverlapActors (Math|Transform|MakeTransform (+ (* (/ (ScalableFloat|GetValueatLevel (Variables|ScalableFloats|GetRow_HomingTargetingMaxRange)) 2.0) (Math|Vector|GetForwardVector _outrotation)) _outlocation) _outrotation) (Math|Vector|MakeVector (/ (ScalableFloat|GetValueatLevel (Variables|ScalableFloats|GetRow_HomingTargetingMaxRange)) 2.0) (ScalableFloat|GetValueatLevel (Variables|ScalableFloats|GetRow_HomingTargetingRadius)) (ScalableFloat|GetValueatLevel (Variables|ScalableFloats|GetRow_HomingTargetingRadius))) (Utilities|Array|MakeArray "ObjectTypeQuery3" "ObjectTypeQuery7") 0 (Utilities|Array|MakeArray (Actor|GetOwner)) :!class /Script/FortniteGame.FortKismetLibrary))
  ;[1 1 1] Finds the closest valid actor
  (bind _asfort_player_pawn (Utilities|Casting|CastToFortPlayerPawn (Actor|GetOwner)))
  (for _array_element _outactors
    (bind _ascharacter (Utilities|Casting|CastToCharacter _array_element))
    (if (> (Class|AbilityFunctions|GetDistanceBetweenActors _ascharacter _asfort_player_pawn) (ScalableFloat|GetValueatLevel (Variables|ScalableFloats|GetRow_HomingTargetingMinRange)))
      (if (not (Team|OnSameTeam _ascharacter _asfort_player_pawn))
        (bind _returnvalue (GameplayTags|HasMatchingGameplayTag(Message) _ascharacter "(TagName=\"Pawn.Athena.NPC.Wildlife\")"))
        (if _returnvalue
          ;[1 1 1] Wildlife
          (LocalVariables|GetTargetWildlife
            (:then
              (if (CallFunction|ACloserThanB? _ascharacter (LocalVariables|GetTargetPlayer))
                _))
            (:else
              _))
          (bind (_outhit _returnvalue_2) (Collision|LineTraceByProfile (Transformation|GetActorLocation) (Physics|GetCenterOfMass (Variables|/Script/Engine.Character|GetCapsuleComponent _ascharacter) "None") "FortProjectileOverlapPawns"))
          (if (not _returnvalue_2)
            (LocalVariables|SetTargetWildlife _ascharacter))
          (else
            (LocalVariables|GetTargetPlayer
              (:then
                (if (CallFunction|ACloserThanB? _ascharacter (LocalVariables|GetTargetPlayer))
                  _))
              (:else
                _))
            (bind (_outhit_2 _returnvalue_3) (Collision|LineTraceByProfile :Start (Transformation|GetActorLocation) :End (Physics|GetCenterOfMass (Variables|/Script/Engine.Character|GetCapsuleComponent _ascharacter) "None") :ProfileName "FortProjectileOverlapPawns" :bTraceComplex false :DrawDebugType "Persistent" :bIgnoreSelf true :TraceColor "(R=1.000000,G=0.000000,B=0.000000,A=1.000000)" :TraceHitColor "(R=0.000000,G=1.000000,B=0.000000,A=1.000000)" :DrawTime 5.0))
            (if (not _returnvalue_3)
              (LocalVariables|SetTargetPlayer _ascharacter)))))))
  ;[1 1 1] Return player; if none return wildlife
  (bind _targetplayer (LocalVariables|GetTargetPlayer)
    (:then
      (Development|PrintString (Utilities|String|Append "Locking On to Character: " (Utilities|GetDisplayName (LocalVariables|GetTargetPlayer))))
      (return (LocalVariables|GetTargetPlayer)))
    (:else
      (bind _targetwildlife (LocalVariables|GetTargetWildlife)
        (:then
          (Development|PrintString (Utilities|String|Append "Locking On to Wildlife: " (Utilities|GetDisplayName (LocalVariables|GetTargetPlayer))))
          (return (LocalVariables|GetTargetWildlife)))
        (:else
          (return 0))))))

(fn GetAffiliationsToShowFor
  (return (LocalVariables|GetAffiliation)))

(fn HandleProjectileStopReturn
  (Networking|SwitchHasAuthority
    (:Authority
      (Utilities|Time|SetTimerbyEvent (EventDispatchers|CreateEvent :!event CallReturnMovement) 1e-05)
      (Transformation|SetRelativeRotation (Variables|Default|GetScene) "0, 0, 0" false false) ; Account for outgoing rotation needing to hit 0 early with a quick hit
    )))

(fn HandleProjectileStop
  (Networking|SwitchHasAuthority
    (:Authority
      (Utilities|Time|SetTimerbyEvent (EventDispatchers|CreateEvent :!event CallMovement) 1e-05))))

(fn HandleThrowingFromVehicle
  (bind _playerpawn (Variables|BPSetVariables|GetPlayerPawn))
  (bind _returnvalue (Vehicle|GetVehicleActor _playerpawn :!class /Script/FortniteGame.FortPlayerPawn))
  (Utilities|IsValid (Variables|BPSetVariables|GetPlayerPawn)
    (:Is%20Valid
      (if (GameplayTags|HasMatchingGameplayTag (Variables|BPSetVariables|GetPlayerPawn) "(TagName=\"Gameplay.Status.InVehicle.Seat.Roofed\")")
        (Utilities|IsValid _returnvalue
          (:Is%20Valid
            (Collision|IgnoreActorwhenMoving (Variables|Collision|GetCapsuleComponent) _returnvalue true)))))))

(fn IncrementThingsHit
  (bind _output_get (Variables|Default|SetTotalOutgoingThingsHit (+ (Variables|Default|GetTotalOutgoingThingsHit) 1)))
  (Variables|Default|SetAudioHitCount _output_get)
  (if (== _output_get 1)
    (Variables|BPSetVariables|SetInitialScalingSpeed (* (ScalableFloat|GetValueatLevel (Variables|ScalableFloats|GetROW_FirstImpact_ReduceMaxSpeedPercent)) (Variables|BPSetVariables|GetInitialScalingSpeed))) ; Reduce the max speed after the first hit, to slow things down a bit after a bounce.
  ))

(fn IsCloserThanReturnRotateThreshold
  (return (<= (Transformation|GetDistanceTo (Variables|BPSetVariables|GetPlayerPawn)) (ScalableFloat|GetValueatLevel (Variables|ScalableFloats|GetRow_ReturnRotateDistanceThreshold)))))

(fn LOSCheck
  (params (TracedActor :type Actor%20Object%20Reference))
  (bind _outhit (Collision|LineTraceByChannel (Transformation|GetActorLocation) (Transformation|GetActorLocation TracedActor) "TraceTypeQuery1" false (Variables|Default|GetActorsToIgnoreForTargeting)))
  (if (Utilities|Equal(Object) (.HitActor (Collision|BreakHitResult _outhit)) TracedActor)
    (return true)
    (else
      (return false))))

(fn OnRep_AudioHitCount
  (Variables|Default|SetbIsAudioMultiHit (> (Variables|Default|GetAudioHitCount) 1)))

(fn OnRep_ReturnToSender
  (bind _self self)
  (bind _projectilemovementcomponent (Variables|Movement|GetProjectileMovementComponent))
  (Utilities|Time|ClearandInvalidateTimerbyHandle (Variables|Default|GetNaturalApexReachedTimerHandle))
  (Development|PrintString "REP NOTIFY: Return To Sender ")
  (Collision|SetCollisionEnabled (Variables|Collision|GetCapsuleComponent) "NoCollision")
  ; Prevents an ensure in SetHomingTarget
  (if (Projectile|IsProjectileBeingKilled :!class /Script/FortniteGame.FortProjectileMovementInterface)
    (Development|PrintString "Projectile was being killed" true true "(R=0.926948,G=0.000000,B=1.000000,A=1.000000)")
    (else
      (Projectile|SetHomingStyle 0 _self "LockOn" false :!class /Script/FortniteGame.FortProjectileMovementInterface)
      (Projectile|SetHomingTarget 0 _self (Variables|BPSetVariables|GetPlayerPawn) 0 :!class /Script/FortniteGame.FortProjectileMovementInterface)
      (Projectile|SetHomingLockTargetDistanceThreshold 0 _self (ScalableFloat|GetValueatLevel (Variables|ScalableFloats|GetRow_ReturnHomingLockTargetDistanceThreshold)) :!class /Script/FortniteGame.FortProjectileMovementInterface)
      (Projectile|SetHomingTargetOffset _self (Variables|HomingValues|GetReturnHomingTargetOffset) :!class /Script/FortniteGame.FortProjectileMovementInterface)
      (Variables|/Script/Engine.ProjectileMovementComponent|SetInitialSpeed (ScalableFloat|GetValueatLevel (Variables|ScalableFloats|GetRow_ReturnStartVelocity)) _projectilemovementcomponent)
      (Variables|/Script/Engine.ProjectileMovementComponent|SetMaxSpeed (ScalableFloat|GetValueatLevel (Variables|ScalableFloats|GetRow_ReturnEndVelocity)) _projectilemovementcomponent)
      (Projectile|SetHomingTurnSpeedMax 0 self (ScalableFloat|GetValueatLevel (Variables|ScalableFloats|GetRow_ReturnMaxTurnSpeed)) :!class /Script/FortniteGame.FortProjectileMovementInterface)
      (Collision|SetCollisionEnabled (Variables|Collision|GetCapsuleComponent) "QueryOnly")
      (CallFunction|ResumeMovement false) ; Do not reset the velocity when returning
      (Variables|Default|SetProjectile%20Homing%20Data (Utilities|Struct|MakeProjectileHomingData "None" (ScalableFloat|GetValueatLevel (Variables|ScalableFloats|GetRow_HomingMinTurnSpeed)) (ScalableFloat|GetValueatLevel (Variables|ScalableFloats|GetRow_HomingMaxTurnSpeed)) (ScalableFloat|GetValueatLevel (Variables|ScalableFloats|GetRow_HomingTurnRampTime)) (ScalableFloat|GetValueatLevel (Variables|ScalableFloats|GetRow_HomingEffectTriggerRange))))
      (CallFunction|StopMovementSpeedScaling))))

(fn PlayRecallSound
  (bind _playerpawn (Variables|BPSetVariables|GetPlayerPawn))
  (Utilities|FlowControl|Sequence
    (:then_0
      ;[1 1 1] Play Recall Whoosh Sound
      (bind _returnvalue (Audio|SpawnSoundAttached (Variables|Default|GetRecallWhooshSound) (Variables|Collision|GetCapsuleComponent)))
      (Utilities|IsValid _playerpawn
        (:Is%20Valid
          (Audio|Parameter|SetBooleanParameter _returnvalue "IsLocal" (Pawn|IsLocallyViewed _playerpawn) :!class /Script/AudioExtensions.AudioParameterControllerInterface))))
    (:then_1
      ;[1 1 1] Stop cast sound so we can then play recall sound. Valid checks just incase of any kind of spam.
      (bind _castinairaudio (Variables|Default|GetCastInAirAudio)
        (:then
          (Audio|Components|Audio|FadeOut _castinairaudio 0.5 0.0 "Linear" :!class /Script/Engine.AudioComponent))
        (:else
          _))
      (Variables|Default|GetRecallAudio
        (:then)
        (:else
          (bind _returnvalue_2 (Audio|SpawnSoundAttached (Variables|Sounds|GetRecallSound) (Variables|Collision|GetCapsuleComponent)))
          (Variables|Default|SetRecallAudio _returnvalue_2))))))

(fn RemedyDamageableShieldHit
  (params (HitBuildingActor :type Building%20Actor%20Object%20Reference))
  (if (GameplayTags|HasMatchingGameplayTag HitBuildingActor "(TagName=\"Trait.Shield.Damageable\")")
    (Utilities|Array|AddUnique (Variables|Default|GetActorsToIgnoreForTargeting) (Game|GetInstigator (Variables|GetHitBuildingActor))) ; Add the target of the shield we hit to the ignore list, we already hit their shield, don't target them again this round.
    (Utilities|Map|Add (Variables|Default|GetRecentlyHit) (Game|GetInstigator (Variables|GetHitBuildingActor)) (Fortnite|Sync|Time|GetServerWorldTimeSeconds))))

(fn Resume_Movement
  (params (SetVelocityToPrevious :type Boolean))
  (Utilities|Time|SetTimerforNextTickbyFunctionName 0 "ResumeSimulation")
  (Collision|SetCollisionEnabled (Variables|Collision|GetCapsuleComponent) "QueryOnly" :!pinned true) ; Make sure the collision is enabled
  (Components|Movement|SetUpdatedComponent (Variables|Movement|GetProjectileMovementComponent) (Variables|Collision|GetCapsuleComponent))
  (if SetVelocityToPrevious
    (Variables|/Script/Engine.MovementComponent|SetVelocity (Variables|BPSetVariables|GetPreviousVelocity) (Variables|Movement|GetProjectileMovementComponent))))

(fn SetHomingVars
  (params (IsBounce :type Boolean))
  (bind _homingtarget (Variables|Default|GetHomingTarget)
    (:then
      (Projectile|SetHomingStyle 0 0 "LockOn" false :!class /Script/FortniteGame.FortProjectileMovementInterface)
      (Utilities|Array|AddUnique (Variables|Default|GetActorsToIgnoreForTargeting) _homingtarget)
      (Networking|SwitchHasAuthority
        (:Authority
          (Projectile|SetHomingTurnSpeedMin 0 self (Math|Float|SelectFloat (ScalableFloat|GetValueatLevel (Variables|ScalableFloats|GetRow_HomingMinTurnSpeed_Bounce)) (ScalableFloat|GetValueatLevel (Variables|ScalableFloats|GetRow_HomingMinTurnSpeed)) (Variables|GetIsBounce)) :!class /Script/FortniteGame.FortProjectileMovementInterface)
          (Projectile|SetHomingTurnSpeedMax 0 self (Math|Float|SelectFloat (ScalableFloat|GetValueatLevel (Variables|ScalableFloats|GetRow_HomingMaxTurnSpeed_Bounce)) (ScalableFloat|GetValueatLevel (Variables|ScalableFloats|GetRow_HomingMaxTurnSpeed)) (Variables|GetIsBounce)) :!class /Script/FortniteGame.FortProjectileMovementInterface)
          (Projectile|SetHomingTimeUntilMaxTurnSpeed 0 self (Math|Float|SelectFloat (ScalableFloat|GetValueatLevel (Variables|ScalableFloats|GetRow_HomingTurnRampTime_Bounce)) (ScalableFloat|GetValueatLevel (Variables|ScalableFloats|GetRow_HomingTurnRampTime)) (Variables|GetIsBounce)) :!class /Script/FortniteGame.FortProjectileMovementInterface)
          (bind _returnvalue (GameplayTags|HasMatchingGameplayTag(Message) (Variables|Default|GetHomingTarget) "(TagName=\"Gameplay.Status.DBNO\")"))
          (Projectile|SetHomingTargetOffset self (Math|Vector|SelectVector (Variables|Default|GetTargetHomingOffset_DBNO) (Variables|Default|GetTarget%20Homing%20Offset) _returnvalue) :!class /Script/FortniteGame.FortProjectileMovementInterface)
          (Projectile|SetHomingTarget 0 self (Variables|Default|GetHomingTarget) 0 :!class /Script/FortniteGame.FortProjectileMovementInterface)
          (Development|PrintString "Set Homing Vars"))))
    (:else)))

(fn ShouldRotateShieldForReturn
  (if (CallFunction|IsCloserThanReturnRotateThreshold)
    (Variables|Default|SetcachedMeshRotation (Transformation|GetWorldRotation (Variables|Default|GetSkelMesh_Weapon)))
    (CallFunction|RotateShieldforReturn)
    (Utilities|Time|ClearandInvalidateTimerbyHandle (Variables|Default|GetTimer_CheckRotateOnReturn))))

(fn ShouldShowSoundIndicator
  (params (PlayerController :type Fort%20Player%20Controller%20Object%20Reference))
  (return (!= PlayerController (Fortnite|GetFortPlayerControllerfromActor (Game|GetInstigator)))))

(fn ConstructionScript
  (bind _capsulecomponent (Variables|Collision|GetCapsuleComponent))
  (Components|Capsule|SetCapsuleHalfHeight _capsulecomponent (ScalableFloat|GetValueatLevel (Variables|ScalableFloats|GetRow_PlayerCollisionHalfHeight)) true)
  (Components|Capsule|SetCapsuleRadius _capsulecomponent (ScalableFloat|GetValueatLevel (Variables|ScalableFloats|GetRow_PlayerCollisionRadius)) true)
  (|CallParentFunction))
